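# Material function for the JarpMilo character body, adapted from ZirconSweep, adding reactive emissive effects for eliminations and progress.
# Two time-stamped pulse copies allow rapid successive eliminations, while a progress band interpolates from previous to current value.
# Pre-skinned position gradients and a mask texture drive progress; scrolling cloud noise and a pattern texture break up the glow.
# It outputs modified emissive, optionally darkened base color, plus a ReactivityMask, with debug switches included.
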MaterialFunction /BRCosmetics/Characters/Player/Male/Medium/Bodies/M_MED_JarpMilo/Materials/MF_M_MED_JarpMilo_Reactivity
graph {
 properties {description=Reacitivity Control for ZirconSweep character}
 nodes [
  n1 = FunctionInput() {inputName=InMaterial, inputType=FunctionInput_MaterialAttributes}
  n2 = GetMaterialAttributes(None=@n1) {attributeGetTypes [#1, #2]}
  n3 = Time()
  n4 = ScalarParameter() {defaultValue=4, group=Reactivity, parameterName=ElimPulseDuration}
  n5 = ScalarParameter() {defaultValue=-100, group=Reactivity, parameterName=ElimPulseStartTime_CCPM, sliderMax=1}
  n6 = Divide(A=StaticSwitchParameter(True=Fmod(A=@n3, B=@n4), False=Subtract(A=@n3, B=@n5)) {group=Reactivity, parameterName=ElimPulse_DEBUG}, B=@n4)
  n7 = OneMinus(None=Saturate(None=@n6))
  n8 = ScalarParameter() {defaultValue=0.65, group=Reactivity, parameterName=ElimPulseProgressExp}
  n9 = ScalarParameter() {defaultValue=0.25, group=Reactivity, parameterName=ElimPulseWidth, sliderMax=0.5}
  n10 = NamedRerouteUsage()
  n11 = OneMinus(None=Power(Base=@n10, Exp=@n8) {constExponent=0.4})
  n12 = PreSkinnedPosition()
  n13 = Time()
  n14 = Multiply(A=Constant3Vector() {constant=<rgba -1 1 1 1>}, B=ScalarParameter() {defaultValue=5, parameterName=Speed})
  n15 = MaterialFunctionCall(
   TextureObject=TextureObject() {samplerType=SAMPLERTYPE_LinearColor, texture=/Game/Global/Textures/Default/Tileable_Patterns/Cloud_Diffuse}
   TextureSize=Multiply(A=ScalarParameter() {defaultValue=70, parameterName=Size}) {constB=0.8}
   WorldPosition=Add(A=VertexInterpolator(VS=@n12).PS, B=Multiply(A=@n13, B=@n14))
  ) {materialFunction=&WorldAlignedTexture}
  n16 = Multiply(
   A=ConstantBiasScale(None=MaterialFunctionCall(Float3=@n15.XYZ Texture) {materialFunction=&BreakOutFloat3Components}.R) {bias=-0.5, scale=2}
   B=ScalarParameter() {defaultValue=0.06, group=Reactivity, parameterName=ElimPulseDistortionAmt, sliderMax=0.1}
  )
  n17 = LinearInterpolate(A=@n9, Alpha=Add(A=@n11, B=Multiply(A=@n11, B=@n16))) {constA=0.15}
  n18 = Power(Base=SphereMask(A=Power(Base=@n6, Exp=@n8) {constExponent=0.78}, B=@n17, Radius=@n9) {attenuationRadius=0.1, hardnessPercent=0}) {constExponent=10}
  n19 = Divide(A=Subtract(A=@n3, B=ScalarParameter() {defaultValue=-100, group=Reactivity, parameterName=ElimPulseStartTime2_CCPM, sliderMax=1}), B=@n4)
  n20 = OneMinus(None=Saturate(None=@n19))
  n21 = Power(Base=SphereMask(A=Power(Base=@n19, Exp=@n8) {constExponent=0.78}, B=@n17, Radius=@n9) {attenuationRadius=0.1, hardnessPercent=0}) {constExponent=10}
  n22 = TextureSampleParameter2D() {
   group=Reactivity
   parameterName=ReactivePattern
   samplerSource=SSM_Wrap_WorldGroupSettings
   samplerType=SAMPLERTYPE_Alpha
   texture=/Game/Characters/Player/Male/Medium/Bodies/M_MED_ZirconSweep/Textures/T_M_MED_ZirconSweep_Body_TechArtFX_Pattern
   channelNames {a="", b="", g="", r=Pattern Mask}
  }
  n23 = Multiply(
   A=Add(A=Multiply(A=@n7, B=@n18), B=Multiply(A=@n20, B=@n21)) {desc=Pulse logic repeated two times to allow for rapid subsequent eliminations}
   B=@n22.Pattern Mask
  )
  n24 = VectorParameter() {group=Reactivity, parameterName=ElimPulseEmissiveColor, defaultValue=<rgba 0.0208 0.6016 1 1>}
  n25 = VectorParameter() {group=Reactivity, parameterName=ElimPulseEmissiveColor2, defaultValue=<rgba 0.0208 0.6016 1 1>}
  n26 = NamedRerouteUsage()
  n27 = Multiply(
   A=Multiply(A=@n23, B=ScalarParameter() {defaultValue=40, group=Reactivity, parameterName=ElimPulseIntensity}) {constB=60}
   B=StaticSwitchParameter(True=LinearInterpolate(A=@n24.RGB, B=@n25.RGB, Alpha=@n26), False=@n24.RGB) {group=Reactivity, parameterName=ElimPulseColorGradientEnabled}
  )
  n28 = ScalarParameter() {defaultValue=5, group=Reactivity, parameterName=Reactive_DEBUG_NumSteps}
  n29 = Floor(None=Fmod(A=Time(), B=Add(A=@n28) {constB=2}))
  n30 = NamedRerouteDeclaration(None=Subtract(A=Time(), B=ScalarParameter() {defaultValue=-100, group=Reactivity, parameterName=ReactiveStartTime_CCPM, sliderMax=1})) {name=ReactivityRelativeTime, nodeColor=<rgba 0 0.0156 1 1>}
  n31 = ScalarParameter() {defaultValue=3.5, group=Reactivity, parameterName=ReactiveDuration_CCPM}
  n32 = LinearInterpolate(
   A=ScalarParameter() {group=Reactivity, parameterName=ReactivePreviousProgress_CCPM, sliderMax=1}
   B=StaticSwitchParameter(True=Divide(A=@n29, B=@n28) {constB=5}, False=ScalarParameter() {group=Reactivity, parameterName=ReactiveProgress_CCPM, sliderMax=1}) {group=Reactivity, parameterName=Reactive_DEBUG}
   Alpha=Saturate(None=Divide(A=@n30, B=@n31))
  )
  n33 = OneMinus(None=NamedRerouteUsage())
  n34 = Constant() {r=1.15}
  n35 = SmoothStep(Min=Add(A=@n32) {constB=0.15}, Max=Subtract(A=@n32) {constB=0.1}, Value=Saturate(None=Multiply(A=@n33, B=@n34)))
  n36 = NamedRerouteDeclaration(None=Power(Base=Saturate(None=@n35), Exp=ScalarParameter() {defaultValue=2, group=Reactivity, parameterName=ReactiveEmissiveGradientExp}) {constExponent=10}) {name=CurrentProgressGradient, nodeColor=<rgba 1 0 0.5859 1>}
  n37 = TextureSampleParameter2D() {
   group=Reactivity
   parameterName=ReactivePattern
   samplerSource=SSM_Wrap_WorldGroupSettings
   samplerType=SAMPLERTYPE_Alpha
   texture=&T_M_MED_ZirconSweep_Body_TechArtFX_Pattern
   channelNames {a="", b="", g="", r=Pattern Mask}
  }
  n38 = Multiply(A=@n36, B=@n37.Pattern Mask) {constA=1}
  n39 = VertexInterpolator(VS=PreSkinnedPosition())
  n40 = Multiply(A=Time(), B=Multiply(A=Constant3Vector() {constant=<rgba -1 1 1 1>}, B=ScalarParameter() {defaultValue=5, parameterName=Speed}))
  n41 = MaterialFunctionCall(Float3=MaterialFunctionCall(
    TextureObject=TextureObject() {samplerType=SAMPLERTYPE_LinearColor, texture=&Cloud_Diffuse}
    TextureSize=ScalarParameter() {defaultValue=70, parameterName=Size}
    WorldPosition=Add(A=@n39.PS, B=@n40)
   ) {materialFunction=&WorldAlignedTexture}.XYZ Texture) {materialFunction=&BreakOutFloat3Components}
  n42 = LinearInterpolate(
   A=@n38
   B=Multiply(A=@n38, B=Power(Base=@n41.R) {constExponent=3})
   Alpha=ScalarParameter() {defaultValue=0.8, group=Reactivity, parameterName=ReactiveNoiseIntensity, sliderMax=1}
  ) {constAlpha=0.9}
  n43 = Multiply(
   A=@n42
   B=Multiply(
    A=VectorParameter() {group=Reactivity, parameterName=ReactiveEmissiveColor, defaultValue=<rgba 1 1 0 1>}.RGB
    B=ScalarParameter() {defaultValue=7, group=Reactivity, parameterName=ReactiveEmissiveIntensity}
   )
  )
  n44 = ScalarParameter() {defaultValue=1, desc=Used in Frontend for smooth looping of reactivity, parameterName=ReactiveMasterAmount_CCPM, sliderMax=1}
  n45 = Add(A=@n2.EmissiveColor, B=Multiply(A=Add(A=@n27, B=@n43), B=@n44))
  n46 = ScalarParameter() {defaultValue=0.5, group=Reactivity, parameterName=ReactiveBaseColorBrightnessScale, sliderMax=1}
  n47 = Multiply(A=@n37.Pattern Mask) {constB=4}
  n48 = TextureSampleParameter2D() {
   group=Reactivity
   parameterName=ReactiveMasks
   samplerSource=SSM_Wrap_WorldGroupSettings
   samplerType=SAMPLERTYPE_Masks
   texture=/Game/Characters/Player/Male/Medium/Bodies/M_MED_ZirconSweep/Textures/T_M_MED_ZirconSweep_Body_TechArtFX_Masks
   channelNames {a="", b=Mirror Mask, g=Gradient Mask, r=Desaturation Mask}
  }
  n49 = Multiply(A=Multiply(A=@n36, B=Saturate(None=@n47)), B=@n48.Desaturation Mask)
  n50 = StaticSwitchParameter(True=LinearInterpolate(A=@n2.BaseColor, B=Multiply(A=@n2.BaseColor, B=@n46), Alpha=Min(A=@n44, B=@n49)) {constB=0}, False=@n2.BaseColor) {group=Reactivity, parameterName=ReactiveBaseColorReduction}
  FunctionOutput(None=StaticSwitchParameter(
    True=SetMaterialAttributes!(MaterialAttributes=@n2.MaterialAttributes, EmissiveColor=@n45, BaseColor=@n50) {attributeSetTypes [#1, #2]}
    False=@n1
   ) {defaultValue=true, group=Reactivity, parameterName=ReactiveEmissiveEnabled, sortPriority=0})
  FunctionOutput(None=NamedRerouteUsage()) {outputName=ReactivityMask, sortPriority=1}
  n51 = ComponentMask(None=VertexInterpolator(VS=PreSkinnedPosition()).PS) {b=true}
  n52 = Constant() {r=110}
  n53 = Saturate(None=Divide(A=Subtract(A=@n51, B=@n52) {constB=20}, B=Constant() {r=130}) {constB=120})
  n54 = VertexInterpolator(VS=PreSkinnedPosition())
  n55 = OneMinus(None=SphereMask(
    A=ComponentMask(None=@n54.PS) {g=true, r=true}
    B=Constant2Vector()
    Radius=ScalarParameter() {defaultValue=25, parameterName=CharacterGradient_RumpRadius, sortPriority=1000}
    Hardness=ScalarParameter() {defaultValue=0.5, parameterName=CharacterGradient_RumpFalloffHardness, sliderMax=1, sortPriority=1000}
   ) {attenuationRadius=30, hardnessPercent=80})
  n56 = MaterialFunctionCall!(
   Local Position
   ArmBalance=ScalarParameter() {defaultValue=0.984, parameterName=Arm Balance, sliderMax=1.5, sortPriority=-10}
   RumpBalance=ScalarParameter() {defaultValue=0.624, parameterName=Rump Balance, sliderMax=1.5, sortPriority=-10}
   Arms Mask=@n55
   Arm_GradientEndPosition
   Arm_GradientBeginPosition
   Arm_GradientDirection
   Rump_GradientEndHeight=ScalarParameter() {parameterName=RumpGradient_FeetHeight, sliderMax=180, sortPriority=64}
   Rump_GradientBeginHeight=ScalarParameter() {defaultValue=165, parameterName=RumpGradient_TopHeight, sliderMax=180, sortPriority=64}
   Radial_CenterHeight=ScalarParameter() {defaultValue=0.6, parameterName=Radial_CenterHeightPercent, sliderMax=1, sortPriority=16}
   Radial_HorizontalCompression=ScalarParameter() {defaultValue=1, parameterName=Radial_HorizontalCompression, sliderMax=2, sliderMin=0.01, sortPriority=16}
   Disable Arm Gradients=StaticBoolParameter() {group=1_Features, parameterName=Disable Arm Gradient}
   Disable Rump Gradient=StaticBoolParameter() {group=1_Features, parameterName=Disable Rump Gradient}
  ) {materialFunction=/Game/Global/MaterialFunctions/TechArt/MF_CharacterGradients}
  n57 = ComponentMask(None=VertexInterpolator(VS=PreSkinnedPosition()).PS) {b=true}
  n58 = ScalarParameter() {defaultValue=11, parameterName=Param_1}
  n59 = Saturate(None=Divide(A=Subtract(A=@n57, B=@n58) {constB=20}, B=ScalarParameter() {defaultValue=90, parameterName=Param_2}) {constB=120})
  n60 = Constant()
  n61 = TextureSample() {samplerType=SAMPLERTYPE_Masks, texture=/BRCosmetics/Characters/Player/Male/Medium/Bodies/M_MED_JarpMilo/Textures/T_JarpMilo_ReactivityMask}
  NamedRerouteDeclaration(None=LinearInterpolate(
    A=LinearInterpolate(A=@n53, B=@n56.RadialGradient, Alpha=@n55)
    B=OneMinus(None=@n59)
    Alpha=StaticSwitchParameter(True=@n60, False=@n61.R) {parameterName=IsHead}
   )) {name=ReactivityProgressGradient, nodeColor=<rgba 1 0.1875 0 1>}
  NamedRerouteDeclaration(None=Multiply(A=@n44, B=Add(A=@n23, B=@n42))) {name=ReactivityMask_OUT, nodeColor=<rgba 0 0.0625 1 1>}
  OneMinus(None=@n10)
 ]
}
