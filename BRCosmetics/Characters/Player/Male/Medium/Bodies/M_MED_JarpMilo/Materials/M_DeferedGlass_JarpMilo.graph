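# Translucent, lit glass material for the JarpMilo player cosmetic's skeletal mesh.
# Base color fakes glass by multiplying normal-offset SceneColor with a grime-masked tint, adding a Fresnel cube-map reflection, then blending in Fresnel color and the body emissive texture, which also feeds Emissive.
# Opacity combines grime, Fresnel and the body mask's green channel; grime and noise drive roughness.
# Mobile swaps in flat edge tint and its own opacity blend.
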Material /BRCosmetics/Characters/Player/Male/Medium/Bodies/M_MED_JarpMilo/Materials/M_DeferedGlass_JarpMilo
graph {
 properties {
  bScreenSpaceReflections=true
  bUseMaterialAttributes=true
  bUsedWithClothing=true
  bUsedWithMeshDeformer=true
  bUsedWithMorphTargets=true
  bUsedWithSkeletalMesh=true
  blendMode=BLEND_Translucent
  pixelDepthOffsetMode=PDOM_Legacy
  translucencyLightingMode=TLM_VolumetricDirectional
  usedShadingModels=Default Lit
 }
 nodes [
  Constant3Vector()
  n1 = VectorParameter() {parameterName=Glass Tint Edge, defaultValue=<rgba 0.3712 0.5035 0.6198 1>}
  n2 = TextureSampleParameter2D() {
   parameterName=GrimeTexture
   samplerType=SAMPLERTYPE_Masks
   texture=/Game/Environments/Apollo/Sets/Hightower_Landmark/Props/Textures/WindowEdgeGrime_M
  }
  n3 = LinearInterpolate(A=@n1.RGB, B=VectorParameter() {parameterName=Glass Tint Center, defaultValue=<rgba 0.3712 0.5035 0.6198 1>}.RGB, Alpha=@n2.R)
  n4 = TextureSample(UVs=Multiply(A=TextureCoordinate(), B=ScalarParameter() {defaultValue=1, parameterName=BlurNormal})) {samplerType=SAMPLERTYPE_Normal, texture=/Game/Global/Textures/Default/Tileable_Patterns/T_SkinBump_N}
  n5 = SceneColor!(OffsetFraction=Multiply(A=ComponentMask(None=@n4.RGB) {g=true, r=true}, B=ScalarParameter() {parameterName=BlurAmount, sortPriority=19})) {inputMode=OffsetFraction}
  n6 = TextureSampleParameterCube!(UVs=ReflectionVectorWS(), Apply View MipBias) {
   parameterName=Cube Map
   samplerType=SAMPLERTYPE_LinearColor
   texture=/Game/Creative/Environments/Meshes/Glass/Textures/T_Glass_Cubemap_01_Desaturated
  }
  n7 = Fresnel(ExponentIn=ScalarParameter() {defaultValue=1, parameterName=Fresnel Exponent})
  n8 = Multiply(A=@n6.RGB, B=@n7)
  n9 = ScalarParameter() {defaultValue=4, parameterName=Cube Map influence}
  n10 = MaterialProxyReplace(Realtime=Add(A=Multiply(A=@n3, B=@n5.RGB), B=Multiply(A=@n8, B=@n9)), MaterialProxy=@n3)
  n11 = ShadingPathSwitch(Default=@n10, Deferred=@n10, Mobile=@n1.RGB)
  n12 = VectorParameter() {parameterName=Fresnel_Color, defaultValue=<rgba 0.03 0.0528 0.1 1>}
  n13 = MaterialFunctionCall(
   Brightness=ScalarParameter() {defaultValue=1.5, parameterName=Fresnel_Brightness}
   Exponent=ScalarParameter() {defaultValue=1, parameterName=Fresnel_Exponent}
  ) {materialFunction=&MF_Fresnel_FX}
  n14 = TextureSample() {texture=/BRCosmetics/Characters/Player/Male/Medium/Bodies/M_MED_JarpMilo/Textures/T_JarpMilo_Body_E}
  n15 = ScalarParameter() {parameterName=colourTint Min}
  n16 = ScalarParameter() {defaultValue=5.036, parameterName=colourTint Max}
  n17 = ScalarParameter() {defaultValue=0.1, parameterName=Roughness_Min}
  n18 = ScalarParameter() {defaultValue=0.1, parameterName=Roughness_Max}
  n19 = TextureSampleParameter2D(UVs=Multiply(A=TextureCoordinate(), B=ScalarParameter() {defaultValue=4, parameterName=Tiling})) {group=Mask Texture, parameterName="Roughness ", sortPriority=10, texture=/Game/Global/Textures/Default/Noise/T_Tiling_Noise_01}
  n20 = ScalarParameter() {defaultValue=-0.6756, parameterName=Roghness Min}
  n21 = ScalarParameter() {defaultValue=0.9605, parameterName=Roughness}
  n22 = TextureSample() {samplerType=SAMPLERTYPE_Masks, texture=/BRCosmetics/Characters/Player/Male/Medium/Bodies/M_MED_JarpMilo/Textures/T_JarpMilo_Body_S}
  n23 = LinearInterpolate(
   A=LinearInterpolate(
    A=ScalarParameter() {parameterName=Fresnel Inner Transparency}
    B=ScalarParameter() {defaultValue=0.85, parameterName=Fresnel Outer Transparency}
    Alpha=@n7
   )
   B=ScalarParameter() {defaultValue=0.5, parameterName=Center Opacity}
   Alpha=@n2.R
  )
  n24 = ShadingPathSwitch(
   Default=@n23
   Deferred=@n23
   Mobile=LinearInterpolate(
    A=ScalarParameter() {defaultValue=0.9, parameterName=Mobile_Opacity0}
    B=ScalarParameter() {defaultValue=0.4, parameterName=Mobile_Opacity1}
    Alpha=NamedRerouteUsage()
   ) {constA=0.8, constB=0.5}
  )
  n25 = Clamp(None=@n13.Fresnel Inverted)
  n26 = TextureSampleParameter2D() {
   group=Mask Texture
   parameterName=BakedNormal
   samplerSource=SSM_Wrap_WorldGroupSettings
   samplerType=SAMPLERTYPE_Normal
   texture=/Engine/EngineMaterials/FlatNormal
  }
  n27 = Normalize(VectorInput=@n26.RGB)
  MakeMaterialAttributes!(
   BaseColor=LinearInterpolate(A=LinearInterpolate(A=@n11, B=@n12.RGB, Alpha=@n13.Fresnel Inverted), B=@n14.RGB, Alpha=LinearInterpolate(A=@n15, B=@n16, Alpha=@n14.B))
   Metallic=ScalarParameter() {parameterName=Metallic}
   Specular=ScalarParameter() {defaultValue=1, parameterName=Specular}
   Roughness=LinearInterpolate(A=ScalarParameter() {defaultValue=1, parameterName=Roughness_Edge}, B=LinearInterpolate(A=@n17, B=@n18, Alpha=@n19.R), Alpha=@n2.R)
   Anisotropy
   EmissiveColor=@n14.RGB
   Opacity=Add(A=LinearInterpolate(A=@n20, B=@n21, Alpha=@n22.G), B=Add(A=@n24, B=@n25))
   OpacityMask
   Normal=@n27
   Tangent
   WorldPositionOffset
   SubsurfaceColor
   ClearCoat
   ClearCoatRoughness
   AmbientOcclusion
   Refraction
   CustomizedUVs_0
   CustomizedUVs_1
   CustomizedUVs_2
   CustomizedUVs_3
   CustomizedUVs_4
   CustomizedUVs_5
   CustomizedUVs_6
   CustomizedUVs_7
   PixelDepthOffset
   ShadingModel
   Displacement
  )
  Multiply()
  n28 = ScalarParameter() {defaultValue=5, parameterName=Mobile_FresnelExponent}
  n29 = ScalarParameter() {defaultValue=0.04, parameterName=Mobile_FresnelBaseReflectFraction}
  NamedRerouteDeclaration(None=LinearInterpolate(
    A=NamedRerouteUsage()
    B=Fresnel(ExponentIn=@n28, BaseReflectFractionIn=@n29)
    Alpha=ScalarParameter() {parameterName=Mobile_ReplaceGrimWithFresnel}
   )) {name=Mobile Opacity, nodeColor=<rgba 0.3828 0 1 1>}
  NamedRerouteDeclaration(None=@n2.R) {name=GrimeTexture, nodeColor=<rgba 0.3125 0 1 1>}
  Transform(None=@n27)
 ]
}
