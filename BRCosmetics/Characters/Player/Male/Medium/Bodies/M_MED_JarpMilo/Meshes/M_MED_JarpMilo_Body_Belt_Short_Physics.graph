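# Physics asset for the JarpMilo male medium body's short belt mesh, defining simulation of its left and right dangling shorts strings.
# Each string is a chain of capsule bodies from a base to three segments, with tight limits at the first joint and looser swing and twist limits below.
# The pelvis, both thighs and a string collider bone supply larger capsules, and these bodies and both string bases join the pelvis with free angular motion.
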PhysicsAsset /BRCosmetics/Characters/Player/Male/Medium/Bodies/M_MED_JarpMilo/Meshes/M_MED_JarpMilo_Body_Belt_Short_Physics
bodies {
 dyn_short_string_01_l [{
   shapeType=Capsule
   length=2.13
   radius=1.0675
   shapeName=dyn_short_string_01_l_capsule
   center=<2.8144 0.0334 0.1045>
   rotation=<rot 0 88.8324 -87.9073>
  }]
 dyn_short_string_01_r [{
   shapeType=Capsule
   length=2.13
   radius=1.0675
   shapeName=dyn_short_string_01_r_capsule
   center=<-2.8096 0.0573 -0.0812>
   rotation=<rot 0 -95.0196 -89.9267>
  }]
 dyn_short_string_02_l [{
   shapeType=Capsule
   length=2.4425
   radius=0.755
   shapeName=dyn_short_string_02_l_capsule
   center=<2.4068 -0.0575 0.112>
   rotation=<rot 0 88.8365 -87.849>
  }]
 dyn_short_string_02_r [{
   shapeType=Capsule
   length=3.9738
   radius=1.1612
   shapeName=dyn_short_string_02_r_capsule
   center=<-1.4089 0.0705 -0.1061>
   rotation=<rot -0.0125 87.4227 -89.6657>
  }]
 dyn_short_string_03_l [{
   shapeType=Capsule
   length=3.816
   radius=1.1447
   shapeName=dyn_short_string_03_l_capsule
   center=<2.8054 -0.0443 0.0585>
   rotation=<rot 0 88.9189 -88.0166>
  }]
 dyn_short_string_03_r [{
   shapeType=Capsule
   length=5.5967
   radius=1.1142
   shapeName=dyn_short_string_03_r_capsule
   center=<-3.32 0.1737 -0.0791>
   rotation=<rot 0 88.5589 -87.968>
  }]
 dyn_short_string_base_l [{shapeType=Capsule, length=0.505, radius=0.505, shapeName=dyn_short_string_base_l_capsule, rotation=<rot 0 -90 0>}]
 dyn_short_string_base_r [{shapeType=Capsule, length=0.505, radius=0.505, shapeName=dyn_short_string_base_r_capsule, rotation=<rot 0 -90 0>}]
 dyn_shorts_string_collider [{shapeType=Capsule, length=10, radius=10, shapeName=dyn_shorts_string_collider_capsule, center=<-8.4523 0.5444 0>, rotation=<rot 0 -90 0>}]
 pelvis [{shapeType=Capsule, length=10, radius=12, center=<-0.2328 -1.6377 0.0116>, rotation=<rot 0.4067 -90 0>}]
 thigh_l [{shapeType=Capsule, length=33.8752, radius=10.1081, shapeName=thigh_l_capsule, center=<-17.7728 0 1.1531>, rotation=<rot 0 90 90>}]
 thigh_r [{shapeType=Capsule, length=33.8752, radius=10.1081, shapeName=thigh_r_capsule, center=<17.7728 0 -1.1531>, rotation=<rot 0 90 -90>}]
}
constraints [
 {
  bone1Name=thigh_r
  bone2Name=pelvis
  swing1LimitDegrees=45
  swing1Motion=Free
  swing2LimitDegrees=45
  swing2Motion=Free
  twistLimitDegrees=45
  twistMotion=Free
 }
 {
  bone1Name=thigh_l
  bone2Name=pelvis
  swing1LimitDegrees=45
  swing1Motion=Free
  swing2LimitDegrees=45
  swing2Motion=Free
  twistLimitDegrees=45
  twistMotion=Free
 }
 {
  bone1Name=dyn_short_string_base_l
  bone2Name=pelvis
  swing1LimitDegrees=45
  swing1Motion=Free
  swing2LimitDegrees=45
  swing2Motion=Free
  twistLimitDegrees=45
  twistMotion=Free
 }
 {bone1Name=dyn_short_string_01_l, bone2Name=dyn_short_string_base_l, swing1LimitDegrees=8, swing2LimitDegrees=8, twistLimitDegrees=8}
 {bone1Name=dyn_short_string_02_l, bone2Name=dyn_short_string_01_l, swing1LimitDegrees=36, swing2LimitDegrees=12, twistLimitDegrees=16}
 {bone1Name=dyn_short_string_03_l, bone2Name=dyn_short_string_02_l, swing1LimitDegrees=36, swing2LimitDegrees=12, twistLimitDegrees=16}
 {
  bone1Name=dyn_short_string_base_r
  bone2Name=pelvis
  swing1LimitDegrees=45
  swing1Motion=Free
  swing2LimitDegrees=45
  swing2Motion=Free
  twistLimitDegrees=45
  twistMotion=Free
 }
 {bone1Name=dyn_short_string_01_r, bone2Name=dyn_short_string_base_r, swing1LimitDegrees=8, swing2LimitDegrees=8, twistLimitDegrees=8}
 {bone1Name=dyn_short_string_02_r, bone2Name=dyn_short_string_01_r, swing1LimitDegrees=36, swing2LimitDegrees=12, twistLimitDegrees=16}
 {bone1Name=dyn_short_string_03_r, bone2Name=dyn_short_string_02_r, swing1LimitDegrees=36, swing2LimitDegrees=12, twistLimitDegrees=16}
 {
  bone1Name=dyn_shorts_string_collider
  bone2Name=pelvis
  swing1LimitDegrees=45
  swing1Motion=Free
  swing2LimitDegrees=45
  swing2Motion=Free
  twistLimitDegrees=45
  twistMotion=Free
 }
]
