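# Physics asset for the JarpMilo male medium player body cosmetic, covering a dangling goggles bone chain (dyn_googles) and the upper torso around it.
# Capsules on spine_05 and both clavicles form the torso shapes, each clavicle constrained to the spine with free rotation.
# The goggles root also rotates freely on spine_05, followed by three chained capsules with tight angular limits that loosen at the last link.
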PhysicsAsset /BRCosmetics/Characters/Player/Male/Medium/Bodies/M_MED_JarpMilo/Meshes/M_MED_JarpMilo_Body_Physics
bodies {
 clavicle_l [{shapeType=Capsule, length=11.2593, radius=8.1848, shapeName=clavicle_l_capsule, center=<8.8896 -0.6411 -1.4319>, rotation=<rot 90 16.4614 0>}]
 clavicle_r [{
   shapeType=Capsule
   length=11.2593
   radius=8.1848
   shapeName=clavicle_r_capsule
   center=<-8.8896 0.6411 1.4319>
   rotation=<rot 90 -163.5386 0>
  }]
 dyn_googles_01 [{shapeType=Capsule, length=6.9468, radius=4.3865, shapeName=dyn_googles_01_capsule, center=<4.2441 0 0>, rotation=<rot 0 0 -90>}]
 dyn_googles_02 [{shapeType=Capsule, length=11.3251, radius=3.9007, shapeName=dyn_googles_02_capsule, center=<4.3044 0 -0.5>, rotation=<rot 0 0 -90>}]
 dyn_googles_03 [{shapeType=Capsule, length=12.2491, radius=5.1513, shapeName=dyn_googles_03_capsule, center=<5.1955 0 -1>, rotation=<rot 0 0 -90>}]
 dyn_googles_root [{shapeType=Capsule, length=0.505, radius=0.505, shapeName=dyn_googles_root_capsule, rotation=<rot 0 -90 0>}]
 spine_05 [{shapeType=Capsule, length=15.6721, radius=12.7733, shapeName=spine_05_capsule, center=<-2.6329 -1.9387 0>, rotation=<rot 0 -109.76 0>}]
}
constraints [
 {
  bone1Name=dyn_googles_root
  bone2Name=spine_05
  swing1LimitDegrees=45
  swing1Motion=Free
  swing2LimitDegrees=45
  swing2Motion=Free
  twistLimitDegrees=45
  twistMotion=Free
 }
 {bone1Name=dyn_googles_01, bone2Name=dyn_googles_root, swing1LimitDegrees=6, swing2LimitDegrees=6, twistLimitDegrees=6}
 {bone1Name=dyn_googles_02, bone2Name=dyn_googles_01, swing1LimitDegrees=6, swing2LimitDegrees=6, twistLimitDegrees=6}
 {bone1Name=dyn_googles_03, bone2Name=dyn_googles_02, swing1LimitDegrees=24, swing2LimitDegrees=24, twistLimitDegrees=24}
 {
  bone1Name=clavicle_l
  bone2Name=spine_05
  swing1LimitDegrees=45
  swing1Motion=Free
  swing2LimitDegrees=45
  swing2Motion=Free
  twistLimitDegrees=45
  twistMotion=Free
 }
 {
  bone1Name=clavicle_r
  bone2Name=spine_05
  swing1LimitDegrees=45
  swing1Motion=Free
  swing2LimitDegrees=45
  swing2Motion=Free
  twistLimitDegrees=45
  twistMotion=Free
 }
]
