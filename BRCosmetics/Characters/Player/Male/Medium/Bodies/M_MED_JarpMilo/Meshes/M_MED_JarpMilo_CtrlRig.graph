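# Control Rig for the JarpMilo player body cosmetic, adding procedural deformation to shorts, drawstrings, belt and flip-flops.
# After the shared DeformRig function, correctives run from thigh attributes, pose readers and toe and calf rotation.
# The belt adds velocity-scaled, root-relative spring chain dynamics, then fits deform bones to Hermite splines.
# RBAN and cloth flags switch drawstring and shorts handling between simulation-supporting adjustments and direct bone posing.
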ControlRigBlueprint /BRCosmetics/Characters/Player/Male/Medium/Bodies/M_MED_JarpMilo/Meshes/M_MED_JarpMilo_CtrlRig
graphs [
 ContainedGraph {nodes [
   n1 = RigVMFunctionEntryNode()
   DISPATCH_RigVMDispatch_ArrayGetAtIndex = RigVMDispatchNode(Array=@n1.Array, Index=@n1.Index) {template="DISPATCH_RigVMDispatch_ArrayGetAtIndex(in Array,in Index,out Element)"}
   &RigUnit_GetTransform(bInitial=False, Item=@DISPATCH_RigVMDispatch_ArrayGetAtIndex.Element)
   n2 = RigVMDispatchNode(
    Index=/Script/RigVM.RigVMFunction_MathIntAdd(B=1, A=@n1.Index) {is=struct RigVMUnitNode, template="Add::Execute(in A,in B,out Result)"}.Result
    Array=@n1.Array
   ) {template="DISPATCH_RigVMDispatch_ArrayGetAtIndex(in Array,in Index,out Element)"}
   &RigUnit_GetTransform(bInitial=False, Item=@n2.Element)
   DISPATCH_RigVMDispatch_ArrayGetNum = RigVMDispatchNode(Array=@n1.Array) {template="DISPATCH_RigVMDispatch_ArrayGetNum(in Array,out Num)"}
   n3 = /Script/RigVM.RigVMFunction_MathIntSub(B=1, A=@DISPATCH_RigVMDispatch_ArrayGetNum.Num) {is=struct RigVMUnitNode, template="Subtract::Execute(in A,in B,out Result)"}
   n4 = &RigUnit_GetRelativeTransformForItem(Parent=(Type=Bone,Name="root"), bChildInitial=False, bParentInitial=False, Child=@DISPATCH_RigVMDispatch_ArrayGetAtIndex.Element)
   RigVMFunctionReturnNode(Transform=RigVMDispatchNode(
     Condition=RigVMDispatchNode(A=@n1.Index, B=@n3.Result) {template="DISPATCH_RigVMDispatch_CoreEquals(in A,in B,out Result)"}.Result
     False=&RigUnit_GetRelativeTransformForItem(Parent=(Type=Bone,Name="root"), bChildInitial=False, bParentInitial=False, Child=@n2.Element).RelativeTransform
     True=/Script/RigVM.RigVMFunction_MathTransformMakeAbsolute(Local=(Rotation=(X=0,Y=0,Z=0,W=1),Translation=(X=1,Y=0,Z=0),Scale3D=(X=1,Y=1,Z=1)), Parent=@n4.RelativeTransform) {is=struct RigVMUnitNode, template="Make Absolute::Execute(in Local,in Parent,out Global)"}.Global
    ) {template="DISPATCH_RigVMDispatch_If(in Condition,in True,in False,out Result)"}.Result)
  ]}
 ContainedGraph {nodes [
   n1 = RigVMFunctionEntryNode()
   SequenceAggregate_4 = &RigVMFunction_Sequence(ExecuteContext=@n1.ExecuteContext)
   SequenceAggregate_4_1 = &RigVMFunction_Sequence(ExecuteContext=@SequenceAggregate_4.B)
   RigVMFunctionReturnNode(A=@SequenceAggregate_4.A, B=@SequenceAggregate_4_1.A, C=@SequenceAggregate_4_1.B)
  ]}
 ContainedGraph {nodes [
   n1 = RigVMFunctionEntryNode()
   Add = &RigVMFunction_MathIntAdd(B=1, A=@n1.Index) {template="Add::Execute(in A,in B,out Result)"}
   DISPATCH_RigVMDispatch_ArrayGetNum = RigVMDispatchNode(Array=@n1.Array) {template="DISPATCH_RigVMDispatch_ArrayGetNum(in Array,out Num)"}
   DISPATCH_RigVMDispatch_CoreEquals = RigVMDispatchNode(
    A=@n1.Index
    B=&RigVMFunction_MathIntSub(B=1, A=@DISPATCH_RigVMDispatch_ArrayGetNum.Num) {template="Subtract::Execute(in A,in B,out Result)"}.Result
   ) {template="DISPATCH_RigVMDispatch_CoreEquals(in A,in B,out Result)"}
   DISPATCH_RigVMDispatch_If = RigVMDispatchNode(False=@Add.Result, Condition=@DISPATCH_RigVMDispatch_CoreEquals.Result, True=@n1.Index) {template="DISPATCH_RigVMDispatch_If(in Condition,in True,in False,out Result)"}
   RigVMFunctionReturnNode(Transform=&RigUnit_GetTransform(
     Space=LocalSpace
     bInitial=False
     Item=RigVMDispatchNode(Index=@DISPATCH_RigVMDispatch_If.Result, Array=@n1.Array) {template="DISPATCH_RigVMDispatch_ArrayGetAtIndex(in Array,in Index,out Element)"}.Element
    ).Transform)
  ]}
 ContainedGraph {nodes [
   n1 = RigVMFunctionEntryNode()
   n2 = &RigVMFunction_MathDoubleMul(B=RigVMVariableNode() {variable=Blend}.Value, A=RigVMVariableNode() {variable=Twist}.Value)
   n3 = /Script/ControlRig.RigUnit_SpringInterpQuaternionV2(
    Current=(X=0,Y=0,Z=0,W=1)
    TargetVelocityAmount=0
    Torque=(X=0,Y=0,Z=0)
    bInitializeFromTarget=False
    bUseCurrentInput=False
    Target=@n1.Global Transform.Rotation
    CriticalDamping=RigVMVariableNode() {variable=Damping}.Value
    Strength=RigVMVariableNode() {variable=Strength}.Value
   ) {
    is=struct RigVMUnitNode
    template="SpringInterp::Execute(in Target,in Strength,in CriticalDamping,in Torque,in bUseCurrentInput,in Current,in TargetVelocityAmount,in bInitializeFromTarget,out Result,out AngularVelocity)"
   }
   n4 = /Script/ControlRig.RigUnit_SpringInterpVectorV2(
    Current=(X=0,Y=0,Z=0)
    Force=(X=0,Y=0,Z=0)
    TargetVelocityAmount=0
    bInitializeFromTarget=False
    bUseCurrentInput=False
    Strength=RigVMVariableNode() {variable=Strength}.Value
    CriticalDamping=RigVMVariableNode() {variable=Damping}.Value
    Target=RigVMVariableNode() {variable=Aim Target}.Value
   ) {
    is=struct RigVMUnitNode
    template="SpringInterp::Execute(in Target,in Strength,in CriticalDamping,in Force,in bUseCurrentInput,in Current,in TargetVelocityAmount,in bInitializeFromTarget,out Result,out Velocity)"
   }
   n5 = /Script/ControlRig.RigUnit_AimBoneMath(
    DebugSettings=(bEnabled=False,Scale=10,WorldOffset=(Rotation=(X=0,Y=0,Z=0,W=1),Translation=(X=0,Y=0,Z=0),Scale3D=(X=1,Y=1,Z=1)))
    InputTransform=(Rotation=(X=0,Y=0,Z=0,W=1),Translation=(X=0,Y=0,Z=0),Scale3D=(X=1,Y=1,Z=1))
    Primary=(Weight=1,Axis=(X=1,Y=0,Z=0),Target=(X=1,Y=0,Z=0),Kind=Location,Space=(Type=Bone,Name="None"))
    Secondary=(Weight=1,Axis=(X=0,Y=0,Z=1),Target=(X=1,Y=0,Z=0),Kind=Direction,Space=(Type=Bone,Name="None"))
    Weight=1
    InputTransform.Translation=@n1.Global Transform.Translation
    Secondary.Target=/Script/RigVM.RigVMFunction_MathQuaternionRotateVector(
     Vector=(X=0,Y=0,Z=1)
     Transform=/Script/RigVM.RigVMFunction_MathQuaternionSlerp(A=@n1.Global Transform.Rotation, T=@n2.Result, B=@n3.Result) {is=struct RigVMUnitNode, template="Interpolate::Execute(in A,in B,in T,out Result)"}.Result
    ) {is=struct RigVMUnitNode, template="Rotate Vector::Execute(in Transform,in Vector,out Result)"}.Result
    Primary.Target=@n4.Result
   ) {is=struct RigVMUnitNode}
   RigVMFunctionReturnNode(
    Result=/Script/RigVM.RigVMFunction_MathTransformLerp(B=@n5.Result, T=RigVMVariableNode() {variable=Blend}.Value, A=RigVMVariableNode() {variable=Global Transform}.Value) {is=struct RigVMUnitNode, template="Interpolate::Execute(in A,in B,in T,out Result)"}.Result
    Spring Result=@n4.Result
   )
  ]}
 New Function_ContainedGraph {nodes [
   n1 = RigVMFunctionEntryNode()
   RigVMFunction_ControlFlowBranch = &RigVMFunction_ControlFlowBranch(ExecuteContext=@n1.ExecuteContext, Condition=RigVMVariableNode() {variable=RBANEnabled}.Value)
   RigVMFunctionReturnNode(ExecuteContext=@RigVMFunction_ControlFlowBranch.Completed)
   &RigUnit_ParentConstraint(
    AdvancedSettings=(RotationOrderForFilter=XZY)
    Child=(Type=Bone,Name="dyn_short_string_01_l")
    Parents=((Item=(Type=Bone,Name="spine_04"),Weight=0.0032),(Item=(Type=Bone,Name="spine_03"),Weight=0.0895),(Item=(Type=Bone,Name="spine_01"),Weight=0.1271),(Item=(Type=Bone,Name="spine_02"),Weight=0.1496),(Item=(Type=Bone,Name="pelvis"),Weight=0.6306))
    Weight=1
    ExecutePin=@RigVMFunction_ControlFlowBranch.False
   )
   > &RigUnit_ParentConstraint(
    AdvancedSettings=(RotationOrderForFilter=XZY)
    Child=(Type=Bone,Name="dyn_short_string_02_l")
    Parents="((Item=(Type=Bone,Name=\"spine_03\"),Weight=0.0072),(Item=(Type=Bone,Name=\"spine_01\"),Weight=0.0111),(Item=(Type=Bone,Name=\"spine_02\"),Weight=0.0132),(Item=(Type=Bone,Name=\"thigh_twist_01_r\"),Weight=0.0759),(Item=(Type=Bone,Name=\"thigh_twist_01_l\"),Weight=0. ...(truncated)"
    Weight=1
   )
   > &RigUnit_ParentConstraint(
    AdvancedSettings=(RotationOrderForFilter=XZY)
    Child=(Type=Bone,Name="dyn_short_string_03_l")
    Parents="((Item=(Type=Bone,Name=\"spine_03\"),Weight=0.0072),(Item=(Type=Bone,Name=\"spine_01\"),Weight=0.0111),(Item=(Type=Bone,Name=\"spine_02\"),Weight=0.0132),(Item=(Type=Bone,Name=\"thigh_twist_01_r\"),Weight=0.0759),(Item=(Type=Bone,Name=\"thigh_twist_01_l\"),Weight=0. ...(truncated)"
    Weight=1
   )
   > &RigUnit_ParentConstraint(
    AdvancedSettings=(RotationOrderForFilter=XZY)
    Child=(Type=Bone,Name="dyn_short_string_01_r")
    Parents="((Item=(Type=Bone,Name=\"dyn_belt_01\"),Weight=0.0011),(Item=(Type=Bone,Name=\"deform_belt_front_01\"),Weight=0.0354),(Item=(Type=Bone,Name=\"deform_belt_pouch\"),Weight=0.1155),(Item=(Type=Bone,Name=\"deform_belt_front_04\"),Weight=0.1202),(Item=(Type=Bone,Name=\" ...(truncated)"
    Weight=1
   )
   > &RigUnit_ParentConstraint(
    AdvancedSettings=(RotationOrderForFilter=XZY)
    Child=(Type=Bone,Name="dyn_short_string_02_r")
    Parents="((Item=(Type=Bone,Name=\"spine_03\"),Weight=0.0072),(Item=(Type=Bone,Name=\"spine_01\"),Weight=0.0111),(Item=(Type=Bone,Name=\"spine_02\"),Weight=0.0132),(Item=(Type=Bone,Name=\"thigh_twist_01_r\"),Weight=0.0759),(Item=(Type=Bone,Name=\"thigh_twist_01_l\"),Weight=0. ...(truncated)"
    Weight=1
   )
   > &RigUnit_ParentConstraint(
    AdvancedSettings=(RotationOrderForFilter=XZY)
    Child=(Type=Bone,Name="dyn_short_string_03_r")
    Parents="((Item=(Type=Bone,Name=\"spine_03\"),Weight=0.0072),(Item=(Type=Bone,Name=\"spine_01\"),Weight=0.0111),(Item=(Type=Bone,Name=\"spine_02\"),Weight=0.0132),(Item=(Type=Bone,Name=\"thigh_twist_01_r\"),Weight=0.0759),(Item=(Type=Bone,Name=\"thigh_twist_01_l\"),Weight=0. ...(truncated)"
    Weight=1
   )
   Get_Animation_Attribute_3 = RigVMDispatchNode(BoneName=None, Default=0, Name=deformRig_l_thigh_up_alpha) {template="DISPATCH_RigDispatch_GetAnimAttribute(in Name,in BoneName,in Default,out Value,out Found)"}
   Get_Animation_Attribute_4 = RigVMDispatchNode(BoneName=None, Default=0, Name=deformRig_r_thigh_up_alpha) {template="DISPATCH_RigDispatch_GetAnimAttribute(in Name,in BoneName,in Default,out Value,out Found)"}
   > Modify Transforms with Mirror_1 = RigVMFunctionReferenceNode(
    Base Modify Data=((Item=(Type=Bone,Name="dyn_short_string_base_l"),Transform=(Rotation=(X=0.0227,Y=0.1285,Z=-0.1722,W=0.9764),Translation=(X=0,Y=0,Z=0),Scale3D=(X=1,Y=1,Z=1))))
    Mirror Modify Data=((Item=(Type=Bone,Name="dyn_short_string_base_r"),Transform=(Rotation=(X=0,Y=0,Z=0,W=1),Translation=(X=0,Y=0,Z=0),Scale3D=(X=1,Y=1,Z=1))))
    Weight=@Get_Animation_Attribute_3.Value
    "Mirror Weight "=@Get_Animation_Attribute_4.Value
   )
   > Modify Transforms with Mirror_2 = RigVMFunctionReferenceNode(
    Base Modify Data=((Item=(Type=Bone,Name="dyn_short_string_base_l"),Transform=(Rotation=(X=0,Y=0,Z=-0.0436,W=0.999),Translation=(X=0,Y=0,Z=0),Scale3D=(X=1,Y=1,Z=1))))
    Mirror Modify Data=((Item=(Type=Bone,Name="dyn_short_string_base_r"),Transform=(Rotation=(X=0,Y=0,Z=0,W=1),Translation=(X=0,Y=0,Z=0),Scale3D=(X=1,Y=1,Z=1))))
   )
   &RigUnit_ParentConstraint(
    AdvancedSettings=(RotationOrderForFilter=XZY)
    Child=(Type=Bone,Name="dyn_short_string_base_l")
    Parents=((Item=(Type=Bone,Name="spine_03"),Weight=0.0566),(Item=(Type=Bone,Name="spine_01"),Weight=0.082),(Item=(Type=Bone,Name="spine_02"),Weight=0.0956),(Item=(Type=Bone,Name="pelvis"),Weight=0.7658))
    Weight=1
    ExecutePin=@RigVMFunction_ControlFlowBranch.True
   )
   > &RigUnit_ParentConstraint(
    AdvancedSettings=(RotationOrderForFilter=XZY)
    Child=(Type=Bone,Name="dyn_short_string_base_r")
    Parents=((Item=(Type=Bone,Name="spine_03"),Weight=0.0572),(Item=(Type=Bone,Name="spine_01"),Weight=0.0826),(Item=(Type=Bone,Name="spine_02"),Weight=0.0963),(Item=(Type=Bone,Name="pelvis"),Weight=0.7639))
    Weight=1
   )
   Get_Animation_Attribute_1 = RigVMDispatchNode(BoneName=None, Default=0, Name=deformRig_r_thigh_up_alpha) {template="DISPATCH_RigDispatch_GetAnimAttribute(in Name,in BoneName,in Default,out Value,out Found)"}
   Get_Animation_Attribute_2 = RigVMDispatchNode(BoneName=None, Default=0, Name=deformRig_l_thigh_up_alpha) {template="DISPATCH_RigDispatch_GetAnimAttribute(in Name,in BoneName,in Default,out Value,out Found)"}
   > Modify Transforms with Mirror = RigVMFunctionReferenceNode(
    Base Modify Data=((Item=(Type=Bone,Name="dyn_short_string_base_l"),Transform=(Rotation=(X=0.0114,Y=0.13,Z=-0.0864,W=0.9877),Translation=(X=0,Y=0,Z=0),Scale3D=(X=1,Y=1,Z=1))))
    Mirror Modify Data=((Item=(Type=Bone,Name="dyn_short_string_base_r"),Transform=(Rotation=(X=0,Y=0,Z=0,W=1),Translation=(X=0,Y=0,Z=0),Scale3D=(X=1,Y=1,Z=1))))
    "Mirror Weight "=@Get_Animation_Attribute_1.Value
    Weight=@Get_Animation_Attribute_2.Value
   )
   Maximum = &RigVMFunction_MathDoubleMax(A=@Get_Animation_Attribute_2.Value, B=@Get_Animation_Attribute_1.Value)
   > &RigUnit_ModifyTransforms(
    ItemToModify=((Item=(Type=Bone,Name="dyn_shorts_string_collider"),Transform=(Rotation=(X=0,Y=0,Z=0,W=1),Translation=(X=4.3798,Y=5.1658,Z=0.7592),Scale3D=(X=1,Y=1,Z=1))))
    WeightMinimum=0
    Weight=@Maximum.Result
   )
  ]}
 New Function_ContainedGraph {nodes [
   Item Array_1 = RigVMDispatchNode(Value=((Type=Bone,Name="deform_belt_front_01"),(Type=Bone,Name="deform_belt_front_02"),(Type=Bone,Name="deform_belt_front_03"),(Type=Bone,Name="deform_belt_front_04"))) {template="DISPATCH_RigVMDispatch_Constant(io Value)"}
   RigUnit_GetTransform = &RigUnit_GetTransform(Item=(Type=Bone,Name="cr_spline_belt_front_01"), bInitial=False)
   RigVMFunction_Sequence_1 = RigVMAggregateNode(ExecuteContext=@n1.ExecuteContext)
   AttachChainToSpline_1 = RigVMFunctionReferenceNode(
    Draw Spline=false
    Draw Spline Thickness=1
    Points=((X=0,Y=0,Z=0),(X=0,Y=0,Z=0),(X=0,Y=0,Z=0),(X=0,Y=0,Z=0))
    Pole Vector Position=(X=0,Y=0,Z=0)
    Primary Axis=(X=1,Y=0,Z=0)
    Samples Per Segment=4
    Secondary Axis=(X=0,Y=0,Z=0)
    Spline Mode=Hermite
    Stretch=true
    Items=@"Item Array_1".Value
    Points.0=@RigUnit_GetTransform.Transform.Translation
    Points.1=&RigUnit_GetTransform(Item=(Type=Bone,Name="cr_spline_belt_front_02"), bInitial=False).Transform.Translation
    Points.2=&RigUnit_GetTransform(Item=(Type=Bone,Name="cr_spline_belt_front_03"), bInitial=False).Transform.Translation
    Points.3=&RigUnit_GetTransform(Item=(Type=Bone,Name="cr_spline_belt_front_04"), bInitial=False).Transform.Translation
    Execute=@RigVMFunction_Sequence_1.B
   )
   Item Array_2 = RigVMDispatchNode(Value=((Type=Bone,Name="deform_belt_rear_01"),(Type=Bone,Name="deform_belt_rear_02"),(Type=Bone,Name="deform_belt_rear_03"),(Type=Bone,Name="deform_belt_rear_04"))) {template="DISPATCH_RigVMDispatch_Constant(io Value)"}
   AttachChainToSpline_2 = RigVMFunctionReferenceNode(
    Draw Spline=false
    Draw Spline Thickness=1
    Points=((X=0,Y=0,Z=0),(X=0,Y=0,Z=0),(X=0,Y=0,Z=0),(X=0,Y=0,Z=0))
    Pole Vector Position=(X=0,Y=0,Z=0)
    Primary Axis=(X=1,Y=0,Z=0)
    Samples Per Segment=4
    Secondary Axis=(X=0,Y=0,Z=0)
    Spline Mode=Hermite
    Stretch=true
    Execute=@AttachChainToSpline_1.Execute
    Items=@"Item Array_2".Value
    Points.0=&RigUnit_GetTransform(Item=(Type=Bone,Name="cr_spline_belt_rear_01"), bInitial=False).Transform.Translation
    Points.1=&RigUnit_GetTransform(Item=(Type=Bone,Name="cr_spline_belt_rear_02"), bInitial=False).Transform.Translation
    Points.2=&RigUnit_GetTransform(Item=(Type=Bone,Name="cr_spline_belt_rear_03"), bInitial=False).Transform.Translation
    Points.3=&RigUnit_GetTransform(Item=(Type=Bone,Name="cr_spline_belt_rear_04"), bInitial=False).Transform.Translation
   )
   &RigUnit_ModifyTransforms(
    ItemToModify="((Item=(Type=Bone,Name=\"deform_belt_rear_01\"),Transform=(Rotation=(X=0,Y=0,Z=0,W=1),Translation=(X=0,Y=0,Z=0),Scale3D=(X=1,Y=1,Z=1))),(Item=(Type=Bone,Name=\"deform_belt_rear_02\"),Transform=(Rotation=(X=0,Y=0,Z=0,W=1),Translation=(X=0,Y=0,Z=0.3),Scale3D=(X= ...(truncated)"
    Weight=1
    WeightMinimum=0
    ExecutePin=@AttachChainToSpline_2.Execute
   )
   > &RigUnit_ParentConstraint(
    AdvancedSettings=(InterpolationType=Average,RotationOrderForFilter=XZY)
    Child=(Type=Bone,Name="deform_belt_pouch")
    Parents=((Item=(Type=Bone,Name="dyn_belt_linear_bounce"),Weight=1))
    Weight=1
   )
   n1 = RigVMFunctionEntryNode()
   Leg ROM = RigVMFunctionReferenceNode(
    ExecuteContext=()
    Include Ankles=true
    Include Distance Checks=true
    Include Knees=true
    Interp Output Speed=5
    Interp Results=true
    Result Alpha=1
    SPR Optional Parent Item=(Type=Bone,Name="None")
   )
   Get_Animation_Attribute_1 = RigVMDispatchNode(BoneName=None, Default=0, Name=deformRig_r_thigh_up_alpha) {template="DISPATCH_RigDispatch_GetAnimAttribute(in Name,in BoneName,in Default,out Value,out Found)"}
   &RigUnit_ModifyTransforms(
    ExecutePin=()
    ItemToModify=((Item=(Type=Bone,Name="None"),Transform=(Rotation=(X=0,Y=0,Z=0,W=1),Translation=(X=0,Y=0,Z=0),Scale3D=(X=1,Y=1,Z=1))))
    WeightMinimum=0
    Weight=@Get_Animation_Attribute_1.Value
   )
   &RigUnit_ModifyTransforms(
    ExecutePin=()
    ItemToModify=((Item=(Type=Bone,Name="deform_belt_pouch"),Transform=(Rotation=(X=0,Y=-0.0489,Z=0,W=-0.9988),Translation=(X=1.0828,Y=0.0791,Z=-1.5),Scale3D=(X=1,Y=1,Z=1))))
    Weight=1
    WeightMinimum=0
   )
   &RigUnit_ModifyTransforms(
    ItemToModify="((Item=(Type=Bone,Name=\"dyn_belt_linear_bounce\"),Transform=(Rotation=(X=0,Y=-0.0489,Z=0,W=-0.9988),Translation=(X=0.8,Y=0,Z=-1),Scale3D=(X=1,Y=1,Z=1))),(Item=(Type=Bone,Name=\"dyn_belt_02\"),Transform=(Rotation=(X=0,Y=0.0175,Z=0,W=0.9998),Translation=(X=0,Y= ...(truncated)"
    Weight=1
    WeightMinimum=0
    ExecutePin=@"Local Chain Dynamics_1".ExecuteContext
   )
   RigVMFunctionReturnNode(ExecuteContext=@RigVMFunction_Sequence_1.C)
   Get_Animation_Attribute = RigVMDispatchNode(BoneName=None, Default=0, Name=deformRig_l_thigh_up_alpha) {template="DISPATCH_RigDispatch_GetAnimAttribute(in Name,in BoneName,in Default,out Value,out Found)"}
   Get_Animation_Attribute_2 = RigVMDispatchNode(BoneName=None, Default=0, Name=deformRig_l_thigh_up_alpha) {template="DISPATCH_RigDispatch_GetAnimAttribute(in Name,in BoneName,in Default,out Value,out Found)"}
   n2 = RigVMDispatchNode(False=0.6, True=0.2, Condition=RigVMVariableNode() {variable=IsCrouch}.Value) {template="DISPATCH_RigVMDispatch_If(in Condition,in True,in False,out Result)"}
   Item Array = RigVMDispatchNode(Value=((Type=Bone,Name="dyn_belt_02"),(Type=Bone,Name="dyn_belt_03"))) {template="DISPATCH_RigVMDispatch_Constant(io Value)"}
   Remap = &RigVMFunction_MathDoubleRemap(SourceMaximum=770, SourceMinimum=650, TargetMaximum=12.5, TargetMinimum=9, Value=RigVMVariableNode() {variable=PawnVelocity}.Value)
   Local Chain Dynamics_1 = RigVMFunctionReferenceNode(
    Blend End=1.2
    Blend Start=0.3
    Damping=0.9
    Chain Weight=@n2.Result
    Bones=@"Item Array".Value
    ExecuteContext=@n4.ExecutePin
    Strength=@Remap.Result
   )
   &RigUnit_ModifyTransforms(
    ItemToModify="((Item=(Type=Bone,Name=\"cr_spline_belt_rear_02\"),Transform=(Rotation=(X=0,Y=0,Z=0,W=-1),Translation=(X=0,Y=-1,Z=0),Scale3D=(X=1,Y=1,Z=1))),(Item=(Type=Bone,Name=\"cr_spline_belt_rear_03\"),Transform=(Rotation=(X=0,Y=0,Z=0.1187,W=-0.9929),Translation=(X=-0.29 ...(truncated)"
    WeightMinimum=0
    Weight=@Get_Animation_Attribute.Value
    ExecutePin=@n5.ExecutePin
   )
   > n3 = &RigUnit_SphericalPoseReader(
    ActiveRegionScaleFactors=(PositiveWidth=3,NegativeWidth=1,PositiveHeight=1.3,NegativeHeight=1.2)
    ActiveRegionSize=0.1
    Debug=(bDrawDebug=false,bDraw2D=False,bDrawLocalAxes=False,DebugScale=25,DebugSegments=20,DebugThickness=0.25)
    DriverAxis=(X=-1,Y=0,Z=0)
    DriverItem=(Type=Bone,Name="thigh_l")
    FalloffRegionScaleFactors=(PositiveWidth=1,NegativeWidth=1,PositiveHeight=0.5,NegativeHeight=0.5)
    FalloffSize=0.57
    FlipHeightScaling=False
    FlipWidthScaling=False
    OptionalParentItem=(Type=Bone,Name="None")
    RotationOffset=(X=130,Y=5,Z=-12)
   )
   > n4 = &RigUnit_ModifyTransforms(
    ItemToModify="((Item=(Type=Bone,Name=\"dyn_belt_03\"),Transform=(Rotation=(X=0,Y=-0.094,Z=0,W=-0.9956),Translation=(X=0,Y=0,Z=0),Scale3D=(X=1,Y=1,Z=1))),(Item=(Type=Bone,Name=\"None\"),Transform=(Rotation=(X=0.1293,Y=-0.0304,Z=0.0704,W=-0.9886),Translation=(X=5.3,Y=1.6,Z=4) ...(truncated)"
    WeightMinimum=0
    Weight=@n3.OutputParam
   )
   &RigUnit_ModifyTransforms(
    ItemToModify="((Item=(Type=Bone,Name=\"dyn_belt_linear_bounce\"),Transform=(Rotation=(X=0.0347,Y=-0.0947,Z=-0.2444,W=-0.9644),Translation=(X=-0.4155,Y=5.4461,Z=-3.3672),Scale3D=(X=1,Y=1,Z=1))),(Item=(Type=Bone,Name=\"dyn_belt_02\"),Transform=(Rotation=(X=0.0863,Y=0.1188,Z=- ...(truncated)"
    WeightMinimum=0
    Weight=@Get_Animation_Attribute_2.Value
    ExecutePin=@RigVMFunction_Sequence_1.A
   )
   > &RigUnit_ParentConstraint(
    AdvancedSettings=(InterpolationType=Average,RotationOrderForFilter=XZY)
    Child=(Type=Bone,Name="cr_spline_belt_front_04")
    Parents=((Item=(Type=Bone,Name="dyn_belt_linear_bounce"),Weight=1))
    Weight=1
   )
   > &RigUnit_ParentConstraint(
    AdvancedSettings=(InterpolationType=Average,RotationOrderForFilter=XZY)
    Child=(Type=Bone,Name="cr_spline_belt_rear_04")
    Parents=((Item=(Type=Bone,Name="dyn_belt_linear_bounce"),Weight=1))
    Weight=1
   )
   > &RigUnit_ParentConstraint(
    AdvancedSettings=(InterpolationType=Average,RotationOrderForFilter=XZY)
    Child=(Type=Bone,Name="cr_spline_belt_front_03")
    Parents=((Item=(Type=Bone,Name="dyn_belt_linear_bounce"),Weight=0.5),(Item=(Type=Bone,Name="pelvis"),Weight=0.5))
    Weight=1
   )
   > n5 = &RigUnit_ParentConstraint(
    AdvancedSettings=(InterpolationType=Average,RotationOrderForFilter=XZY)
    Child=(Type=Bone,Name="cr_spline_belt_rear_03")
    Parents=((Item=(Type=Bone,Name="dyn_belt_linear_bounce"),Weight=0.5),(Item=(Type=Bone,Name="pelvis"),Weight=0.5))
    Weight=1
   )
  ]}
 New Function_ContainedGraph {nodes [
   RigVMFunctionEntryNode()
   > Get Delta Ref = RigVMFunctionReferenceNode(Item=(Type=Bone,Name="ball_l"))
   > Get Delta Ref_1 = RigVMFunctionReferenceNode(Item=(Type=Bone,Name="ball_r"))
   RigVMFunction_MathQuaternionFromRotatorV2 = /Script/RigVM.RigVMFunction_MathQuaternionFromRotatorV2(Value=@"Get Delta Ref".Result.Rotation) {is=struct RigVMUnitNode, template="Cast::Execute(in Value,out Result)"}
   RigVMFunction_MathQuaternionToEuler = &RigVMFunction_MathQuaternionToEuler(RotationOrder=ZYX, Value=@RigVMFunction_MathQuaternionFromRotatorV2.Result)
   Remap = &RigVMFunction_MathDoubleRemap(SourceMaximum=30, SourceMinimum=0, TargetMaximum=1, TargetMinimum=0, Value=@RigVMFunction_MathQuaternionToEuler.Result.Z)
   Add = &RigVMFunction_MathDoubleAdd(B=0, A=@"Get Delta Ref_1".Result.Rotation.Yaw)
   > Modify Transforms with Mirror = RigVMFunctionReferenceNode(
    Base Modify Data=((Item=(Type=Bone,Name="dyn_flip_flopper_l"),Transform=(Rotation=(X=0,Y=0,Z=0.1305,W=0.9914),Translation=(X=0,Y=0,Z=0),Scale3D=(X=1,Y=1,Z=1))))
    Mirror Modify Data=((Item=(Type=Bone,Name="dyn_flip_flopper_r"),Transform=(Rotation=(X=0,Y=0,Z=0,W=1),Translation=(X=0,Y=0,Z=0),Scale3D=(X=1,Y=1,Z=1))))
    Weight=@Remap.Result
    "Mirror Weight "=&RigVMFunction_MathDoubleRemap(SourceMaximum=-30, SourceMinimum=0, TargetMaximum=1, TargetMinimum=0, Value=@Add.Result).Result
   )
   > RigVMFunctionReturnNode()
  ]}
 New Function_ContainedGraph {nodes [
   RigVMFunctionEntryNode()
   > SequenceAggregate_4 = RigVMAggregateNode()
   SequenceAggregate = &RigVMFunction_Sequence(ExecuteContext=@SequenceAggregate_4.A)
   n1 = RigVMVariableNode() {variable=Bones}
   n2 = RigVMDispatchNode(ExecuteContext=@SequenceAggregate.A, Array=@n1.Value) {template="DISPATCH_RigVMDispatch_ArrayIterator(in Array,out Element,out Index,out Count,out Ratio)"}
   > Get Root Relative Transform = RigVMFunctionReferenceNode(Item=@n2.Element)
   Get Next Global Transform = RigVMCollapseNode(Index=@n2.Index, Array=@n1.Value)
   n3 = &RigUnit_AimBoneMath(
    DebugSettings=(bEnabled=False,Scale=10,WorldOffset=(Rotation=(X=0,Y=0,Z=0,W=1),Translation=(X=0,Y=0,Z=0),Scale3D=(X=1,Y=1,Z=1)))
    Primary=(Weight=1,Axis=(X=1,Y=0,Z=0),Target=(X=1,Y=0,Z=0),Kind=Location,Space=(Type=Bone,Name="None"))
    Secondary=(Weight=1,Axis=(X=0,Y=0,Z=1),Target=(X=1,Y=0,Z=0),Kind=Direction,Space=(Type=Bone,Name="None"))
    Weight=1
    Primary.Target=@"Get Next Global Transform".Transform.Translation
    InputTransform=@"Get Root Relative Transform".Transform
    Secondary.Target=&RigVMFunction_MathQuaternionRotateVector(Vector=(X=0,Y=0,Z=1), Transform=@"Get Root Relative Transform".Transform.Rotation) {template="Rotate Vector::Execute(in Transform,in Vector,out Result)"}.Result
   )
   > RigVMDispatchNode(Name=ChainDynamics:Current, NameSpace=Self, Value=@n3.Result, Item=@n2.Element) {template="DISPATCH_RigDispatch_SetMetadata(in Item,in Name,in NameSpace,in Value,out Success)"}
   n4 = RigVMDispatchNode(ExecuteContext=@SequenceAggregate.B, Array=RigVMVariableNode() {variable=Bones}.Value) {template="DISPATCH_RigVMDispatch_ArrayIterator(in Array,out Element,out Index,out Count,out Ratio)"}
   n5 = RigVMDispatchNode(
    Default=(Rotation=(X=0,Y=0,Z=0,W=1),Translation=(X=0,Y=0,Z=0),Scale3D=(X=1,Y=1,Z=1))
    Name=ChainDynamics:Current
    NameSpace=Self
    Item=@n4.Element
   ) {template="DISPATCH_RigDispatch_GetMetadata(in Item,in Name,in NameSpace,in Default,out Value,out Found)"}
   > Set Root Relative Transform = RigVMFunctionReferenceNode(Transform=@n5.Value, Item=@n4.Element)
   > DISPATCH_RigDispatch_SetMetadata = RigVMDispatchNode(
    Name=ChainDynamics:Blend
    NameSpace=Self
    Item=@n4.Element
    Value=/Script/RigVM.RigVMFunction_MathDoubleLerp(T=@n4.Ratio, A=RigVMVariableNode() {variable=Blend Start}.Value, B=RigVMVariableNode() {variable=Blend End}.Value) {is=struct RigVMUnitNode, template="Interpolate::Execute(in A,in B,in T,out Result)"}.Result
   ) {template="DISPATCH_RigDispatch_SetMetadata(in Item,in Name,in NameSpace,in Value,out Success)"}
   n6 = RigVMDispatchNode(ExecuteContext=@SequenceAggregate_4.B, Array=RigVMVariableNode() {variable=Bones}.Value) {template="DISPATCH_RigVMDispatch_ArrayIterator(in Array,out Element,out Index,out Count,out Ratio)"}
   > Get Root Relative Transform_1 = RigVMFunctionReferenceNode(Item=@n6.Element)
   Get Next Local Transform = RigVMCollapseNode(Index=@n6.Index, Array=RigVMVariableNode() {variable=Bones}.Value)
   DISPATCH_RigDispatch_GetMetadata = RigVMDispatchNode(
    Default=(Rotation=(X=0,Y=0,Z=0,W=1),Translation=(X=0,Y=0,Z=0),Scale3D=(X=1,Y=1,Z=1))
    Name=ChainDynamics:Current
    NameSpace=Self
    Item=@n6.Element
   ) {template="DISPATCH_RigDispatch_GetMetadata(in Item,in Name,in NameSpace,in Default,out Value,out Found)"}
   Aim Spring = RigVMCollapseNode(
    Blend=RigVMDispatchNode(Default=1, Name=ChainDynamics:Blend, NameSpace=Self, Item=@n6.Element) {template="DISPATCH_RigDispatch_GetMetadata(in Item,in Name,in NameSpace,in Default,out Value,out Found)"}.Value
    Global Transform=@"Get Root Relative Transform_1".Transform
    Aim Target=/Script/RigVM.RigVMFunction_MathTransformTransformVector(
     Location=@"Get Next Local Transform".Transform.Translation
     Transform=/Script/RigVM.RigVMFunction_MathTransformMake(
      Rotation=@DISPATCH_RigDispatch_GetMetadata.Value.Rotation
      Scale=@DISPATCH_RigDispatch_GetMetadata.Value.Scale3D
      Translation=@"Get Root Relative Transform_1".Transform.Translation
     ) {is=struct RigVMUnitNode}.Result
    ) {is=struct RigVMUnitNode}.Result
    Strength=RigVMVariableNode() {variable=Strength}.Value
    Damping=RigVMVariableNode() {variable=Damping}.Value
    Twist=RigVMVariableNode() {variable=Blend Twist}.Value
   )
   > /Script/ControlRig.RigUnit_SetRelativeTransformForItem(
    Parent=(Type=Bone,Name="root")
    bParentInitial=False
    bPropagateToChildren=True
    Value=@"Aim Spring".Result
    Child=@n6.Element
    Weight=RigVMVariableNode() {variable=Chain Weight}.Value
   ) {
    is=struct RigVMUnitNode
    template="Set Relative Transform::Execute(in Child,in Parent,in bParentInitial,in Value,in Weight,in bPropagateToChildren)"
   }
   > /Script/RigVM.RigVMFunction_DebugTransformMutableNoSpace(
    Color=(R=1,G=1,B=0,A=1)
    DebugDrawSettings=(DepthPriority=SDPG_Foreground,Lifetime=-1)
    Mode=Point
    Scale=5
    Thickness=1
    Transform=(Rotation=(X=0,Y=0,Z=0,W=1),Translation=(X=0,Y=0,Z=0),Scale3D=(X=1,Y=1,Z=1))
    WorldOffset=(Rotation=(X=0,Y=0,Z=0,W=1),Translation=(X=0,Y=0,Z=0),Scale3D=(X=1,Y=1,Z=1))
    bEnabled=True
    Transform.Translation=@"Aim Spring".Spring Result
   ) {is=struct RigVMUnitNode}
   RigVMFunctionReturnNode(ExecuteContext=@SequenceAggregate_4.C)
   &RigUnit_GetTransform(bInitial=False, Item=@n6.Element)
   &RigUnit_GetTransform(bInitial=False, Item=@n2.Element)
   Set Root Relative Transform_1 = RigVMFunctionReferenceNode(ExecuteContext=(), Transform=@"Aim Spring".Result, Item=@n6.Element)
   &RigUnit_SetTransform(ExecutePin=(), Weight=1, bInitial=False, Value=@"Aim Spring".Result, Item=@n6.Element)
   &RigUnit_SetTransform(ExecutePin=(), Weight=1, bInitial=False, Value=@n5.Value, Item=@n4.Element)
  ]}
 New Function_ContainedGraph {nodes [
   n1 = RigVMFunctionEntryNode()
   > RigVMFunctionReturnNode(Transform=&RigUnit_GetRelativeTransformForItem(Parent=(Type=Bone,Name="root"), bChildInitial=False, bParentInitial=False, Child=@n1.Item).RelativeTransform)
  ]}
 New Function_ContainedGraph {nodes [
   n1 = RigVMFunctionEntryNode()
   > &RigUnit_SetRelativeTransformForItem(Parent=(Type=Bone,Name="root"), Weight=1, bParentInitial=False, bPropagateToChildren=True, Child=@n1.Item, Value=@n1.Transform) {template="Set Relative Transform::Execute(in Child,in Parent,in bParentInitial,in Value,in Weight,in bPropagateToChildren)"}
   > RigVMFunctionReturnNode()
   &RigUnit_SetTransform(ExecutePin=(), Weight=1, bInitial=False, Item=@n1.Item, Value=@n1.Transform)
  ]}
 New Function_ContainedGraph {nodes [
   n1 = RigVMFunctionEntryNode()
   > RigVMFunction_Sequence = RigVMAggregateNode()
   Get Delta Ref = RigVMFunctionReferenceNode(Item=(Type=Bone,Name="calf_l"), ExecuteContext=@RigVMFunction_Sequence.A)
   Remap = &RigVMFunction_MathDoubleRemap(SourceMaximum=130, SourceMinimum=90, TargetMaximum=1, TargetMinimum=0, Value=@"Get Delta Ref".Result.Rotation.Yaw)
   > RigVMVariableNode(Value=@Remap.Result) {variable=Get Delta Ref_Result_Rotation_Yaw}
   > Get Delta Ref_1 = RigVMFunctionReferenceNode(Item=(Type=Bone,Name="calf_r"))
   > RigVMVariableNode(Value=&RigVMFunction_MathDoubleRemap(SourceMaximum=130, SourceMinimum=90, TargetMaximum=1, TargetMinimum=0, Value=@"Get Delta Ref_1".Result.Rotation.Yaw).Result) {variable=Get Delta Ref_1_Result_Rotation_Yaw}
   RigVMFunction_ControlFlowBranch = &RigVMFunction_ControlFlowBranch(ExecuteContext=@RigVMFunction_Sequence.B, Condition=@n1.Cloth Enabled)
   Modify Transforms with Mirror = RigVMFunctionReferenceNode(
    Base Modify Data=((Item=(Type=Bone,Name="dyn_calf_cloth_collider_l"),Transform=(Rotation=(X=0,Y=0,Z=0,W=1),Translation=(X=0,Y=5,Z=0),Scale3D=(X=1,Y=1,Z=1))))
    Mirror Modify Data=((Item=(Type=Bone,Name="dyn_calf_cloth_collider_r"),Transform=(Rotation=(X=0,Y=0,Z=0,W=1),Translation=(X=0,Y=0,Z=0),Scale3D=(X=1,Y=1,Z=1))))
    ExecuteContext=@RigVMFunction_ControlFlowBranch.True
    Weight=RigVMVariableNode() {variable=Get Delta Ref_Result_Rotation_Yaw}.Value
    "Mirror Weight "=RigVMVariableNode() {variable=Get Delta Ref_1_Result_Rotation_Yaw}.Value
   )
   Modify Transforms with Mirror_1 = RigVMFunctionReferenceNode(
    Base Modify Data="((Item=(Type=Bone,Name=\"dyn_shorts_base_l\"),Transform=(Rotation=(X=0,Y=0,Z=0,W=1),Translation=(X=5,Y=2,Z=0),Scale3D=(X=1,Y=1,Z=1))),(Item=(Type=Bone,Name=\"dyn_shorts_03_l\"),Transform=(Rotation=(X=0,Y=0,Z=-0.1736,W=0.9848),Translation=(X=-3,Y=0,Z=0),Scale3D ...(truncated)"
    Mirror Modify Data="((Item=(Type=Bone,Name=\"dyn_shorts_base_r\"),Transform=(Rotation=(X=0,Y=0,Z=0,W=1),Translation=(X=0,Y=0,Z=0),Scale3D=(X=1,Y=1,Z=1))),(Item=(Type=Bone,Name=\"dyn_shorts_03_r\"),Transform=(Rotation=(X=0,Y=0,Z=0,W=1),Translation=(X=0,Y=0,Z=0),Scale3D=(X=1,Y=1,Z= ...(truncated)"
    Weight=RigVMVariableNode() {variable=Get Delta Ref_Result_Rotation_Yaw}.Value
    "Mirror Weight "=RigVMVariableNode() {variable=Get Delta Ref_1_Result_Rotation_Yaw}.Value
    ExecuteContext=@RigVMFunction_ControlFlowBranch.False
   )
   RigVMFunctionReturnNode(ExecuteContext=@RigVMFunction_Sequence.C)
  ]}
 RigVMFunctionLibrary {nodes [
   Deform Belt = RigVMCollapseNode(ExecuteContext=())
   Deform Cloth Calf = RigVMCollapseNode(Cloth Enabled=true, ExecuteContext=())
   Deform Flip Flops = RigVMCollapseNode(ExecuteContext=())
   Deform Short Strings = RigVMCollapseNode(ExecuteContext=())
   Get Root Relative Transform = RigVMCollapseNode(ExecuteContext=(), Item=(Type=None,Name="None"))
   Local Chain Dynamics = RigVMCollapseNode(Bones=(), ExecuteContext=())
   Set Root Relative Transform = RigVMCollapseNode(ExecuteContext=(), Item=(Type=Bone,Name="None"), Transform=(Rotation=(X=0,Y=0,Z=0,W=1),Translation=(X=0,Y=0,Z=0),Scale3D=(X=1,Y=1,Z=1)))
  ]}
 RigVMFunction_Sequence_1_ContainedGraph {nodes [
   n1 = RigVMFunctionEntryNode()
   RigVMFunction_Sequence = &RigVMFunction_Sequence(ExecuteContext=@n1.ExecuteContext)
   n2 = &RigVMFunction_Sequence(ExecuteContext=@RigVMFunction_Sequence.B)
   RigVMFunctionReturnNode(A=@RigVMFunction_Sequence.A, B=@n2.A, C=@n2.B)
  ]}
 RigVMFunction_Sequence_2_ContainedGraph {nodes [
   n1 = RigVMFunctionEntryNode()
   n2 = &RigVMFunction_Sequence(ExecuteContext=@n1.ExecuteContext)
   n3 = &RigVMFunction_Sequence(ExecuteContext=@n2.B)
   RigVMFunctionReturnNode(A=@n2.A, B=@n3.A, C=@n3.B)
  ]}
 RigVMModel {nodes [
   RigUnit_BeginExecution = &RigUnit_BeginExecution()
   FN_DeformRig_FNC = RigVMFunctionReferenceNode(
    DeformRig_ Curve="(deformRig_l_upperarm_up_alpha_72_#1=true,deformRig_r_upperarm_up_alpha_73_#2=true,deformRig_l_upperarm_fr_alpha_74_#3=False,deformRig_r_upperarm_fr_alpha_75_CB51B80C ...(truncated)"
    DeformRig_Bool="(upperarm_1_#4=True,elbow_3_#5=True,wrist_7_#6=True,groin_8_#7=True,glute_17_#8=True,knee_9_40C408324CA261 ...(truncated)"
    DeformRig_Debug=(upperarm_debug_18_#4=False,wrist_debug_20_#6=False,lat_debug_26_#9=False,groinGlute_debug_27_#7=False)
    DeformRig_Value="(upperarm_up_18_#10=(Rotation=(X=0.0,Y=-0.0,Z=0.0,W=1.0),Translation=(X=0.0,Y=0.0,Z=8.0),Scale3D=(X=0.70,Y=1.0,Z=1.0)),upperarm_fwd_fr_19_#11=(Rotation=(X=0.0,Y=-0.0,Z=0.0,W=1.0),Translation=(X=0.0, ...(truncated)"
    ExecuteContext=@RigUnit_BeginExecution.ExecutePin
   )
   > Deform Short Strings = RigVMFunctionReferenceNode()
   > Deform Belt = RigVMFunctionReferenceNode()
   > Deform Flip Flops = RigVMFunctionReferenceNode()
   > Deform Cloth Calf = RigVMFunctionReferenceNode(Cloth Enabled=RigVMVariableNode() {variable=ClothEnabled}.Value)
   /Script/FortniteGame.RigUnit_FortDeformRig(
    Bool=(bUpperArm=true,bElbow=true,bWrist=true,bGroin=true,bGlute=true,bKnee=true,bPec=true,bBicep=False,bLat=False)
    Curve="(bLUpperArmUpAlpha=False,bRUpperArmUpAlpha=False,bLUpperArmFrAlpha=False,bRUpperArmFrAlpha=False,bLUpperArmBkAlpha=False,bRUpperArmBkAlpha=False,bLWristUpAlpha=False,bRWristUpAlpha=False,bLWristDnAlpha=False,bRWristDnAlpha=False,bLKneeAlpha=False,bRKneeAlp ...(truncated)"
    Debug=(bUpperArmDebug=False,bWristDebug=False,bLatDebug=False,bGroinGluteDebug=False)
    ExecutePin=()
    Value="(UpperArmUp=(Rotation=(X=0,Y=0,Z=0,W=1),Translation=(X=0,Y=0,Z=8),Scale3D=(X=1,Y=1,Z=1)),UpperArmFwdFr=(Rotation=(X=0,Y=0,Z=0,W=1),Translation=(X=0,Y=10,Z=0),Scale3D=(X=1,Y=1,Z=1)),UpperArmFwdBk=(Rotation=(X=0,Y=0,Z=0,W=1),Translation=(X=0,Y=-6,Z=0),Scale3 ...(truncated)"
   ) {is=struct RigVMUnitNode}
  ]}
]
hierarchy [
 root {loc=<0 0 0>, rot=<0 0 0 1>}
 pelvis {parent=root, loc=<0 -0.7071 93.1191>, rot=<0.0127 0.707 -0.0127 -0.707>}
 spine_01 {parent=pelvis, loc=<6.7494 -0.6388 0>, rot=<0 0 0.034 -0.9994>}
 spine_02 {parent=spine_01, loc=<7.2518 -0.035 0>, rot=<0 0 0.0238 -0.9997>}
 spine_03 {parent=spine_02, loc=<6.0104 0 0>, rot=<0 0 0.0149 -0.9999>}
 spine_04 {parent=spine_03, loc=<8.1024 0 0>, rot=<0 0 -0.0132 -0.9999>}
 spine_05 {parent=spine_04, loc=<10.0146 0 0>, rot=<0 0 -0.0015 -1>}
 neck_01 {parent=spine_05, loc=<11.3138 0.1529 0>, rot=<0 0 -0.2003 -0.9797>}
 neck_02 {parent=neck_01, loc=<4.5627 0 0>, rot=<0 0 0.0504 -0.9987>}
 head {parent=neck_02, loc=<5.3914 0 0>, rot=<0 0 0.115 -0.9934>}
 dyn_googles_root {parent=neck_02, loc=<0.8101 -4.9435 0>, rot=<0.4146 0.5728 -0.5728 -0.4146>}
 dyn_googles_01 {parent=dyn_googles_root, loc=<0 0 0>, rot=<0 0 0 1>}
 dyn_googles_02 {parent=dyn_googles_01, loc=<8.5416 0 0>, rot=<0 0 0 1>}
 dyn_googles_03 {parent=dyn_googles_02, loc=<4.8726 0 0>, rot=<0 0 0 1>}
 attach_cape {parent=spine_05, loc=<9.017 -11.714 0>, rot=<0.0616 -0.7044 -0.0616 -0.7044>}
 cape {parent=attach_cape, type=Null, loc=<0 2.2239 -0.7814>, rot=<0 0 0 1>}
 attach_backpack {parent=spine_05, loc=<-0.123 -13.473 0>, rot=<-0.0093 -0.707 0.0093 -0.707>}
 Backpack_BR {parent=attach_backpack, type=Null, loc=<0 1.3216 -2.2626>, rot=<0 0 0 1>}
 clavicle_l {parent=spine_05, loc=<4.3575 -0.1466 -2.3223>, rot=<0.0465 -0.7289 -0.009 -0.6829>}
 upperarm_l {parent=clavicle_l, loc=<15.5111 0 0>, rot=<0.0226 -0.3234 -0.0088 -0.9459>}
 lowerarm_l {parent=upperarm_l, loc=<26.6514 0 0>, rot=<0.0391 -0.0099 -0.2453 -0.9686>}
 hand_l {parent=lowerarm_l, loc=<22.3491 0 0>, rot=<-0.7081 -0.0507 0.0975 -0.6975>}
 thumb_01_l {parent=hand_l, loc=<1.3118 -0.3191 -2.459>, rot=<0.6546 -0.4042 0.0664 -0.6354>}
 thumb_02_l {parent=thumb_01_l, loc=<5.9336 -0.0448 0.0401>, rot=<-0.0037 0.0213 0.298 -0.9543>}
 thumb_03_l {parent=thumb_02_l, loc=<2.2223 0.3682 -0.0907>, rot=<0.0118 -0.013 0.0066 -0.9998>}
 index_metacarpal_l {parent=hand_l, loc=<1.6535 0.0183 -2.96>, rot=<0.0158 -0.0124 -0.0173 -0.9996>}
 index_01_l {parent=index_metacarpal_l, loc=<8.3008 0.0357 -0.0524>, rot=<-0.0044 0.0485 0.0924 -0.9945>}
 index_02_l {parent=index_01_l, loc=<3.3587 -0.3004 0.0452>, rot=<-0.002 0.0027 0.3225 -0.9466>}
 index_03_l {parent=index_02_l, loc=<2.5578 -0.123 -0.0252>, rot=<-0.002 0.0067 0.2558 -0.9667>}
 middle_metacarpal_l {parent=hand_l, loc=<2.1856 -0.1725 -0.5984>, rot=<0.006 0.0168 -0.0329 -0.9993>}
 middle_01_l {parent=middle_metacarpal_l, loc=<7.3275 -0.2447 0.0129>, rot=<-0.0208 0.0042 0.0858 -0.9961>}
 middle_02_l {parent=middle_01_l, loc=<4.0123 0.1401 0.1175>, rot=<-0.0218 0.0258 0.3512 -0.9357>}
 middle_03_l {parent=middle_02_l, loc=<2.6077 0.0534 0.1373>, rot=<0.0082 0.0188 0.1708 -0.9851>}
 ring_metacarpal_l {parent=hand_l, loc=<2.2074 -0.3625 1.3638>, rot=<-0.0284 0.0722 -0.0458 -0.9959>}
 ring_01_l {parent=ring_metacarpal_l, loc=<7.0295 -0.6197 -0.1446>, rot=<-0.0489 -0.0252 0.1892 -0.9804>}
 ring_02_l {parent=ring_01_l, loc=<3.6174 0.4647 -0.1027>, rot=<0 0 0.2806 -0.9598>}
 ring_03_l {parent=ring_02_l, loc=<2.4152 0.114 0.0199>, rot=<0 0 0.1189 -0.9929>}
 pinky_metacarpal_l {parent=hand_l, loc=<1.7616 -0.9753 3.3814>, rot=<-0.0397 0.0814 -0.069 -0.9935>}
 pinky_01_l {parent=pinky_metacarpal_l, loc=<7.1709 -0.4831 -0.1094>, rot=<-0.0654 -0.0035 0.2242 -0.9723>}
 pinky_02_l {parent=pinky_01_l, loc=<2.8274 0.0893 -0.1043>, rot=<0.0037 0.0011 0.2854 -0.9584>}
 pinky_03_l {parent=pinky_02_l, loc=<2.2046 0.0307 -0.0078>, rot=<-0.0002 0.0098 0.1445 -0.9895>}
 deform_wrist_dn_base_l {parent=hand_l, loc=<0 -2.2342 0>, rot=<0 0 0 -1>}
 deform_wrist_dn_l {parent=deform_wrist_dn_base_l, loc=<0 0 0>, rot=<0 0 0 1>}
 deform_wrist_up_base_l {parent=hand_l, loc=<0 2.2682 0>, rot=<0 0 0 -1>}
 deform_wrist_up_l {parent=deform_wrist_up_base_l, loc=<0 0 0>, rot=<0 0 0 1>}
 lowerarm_twist_02_l {parent=lowerarm_l, loc=<9.6199 0 0>, rot=<0 0 0 -1>}
 lowerarm_twist_01_l {parent=lowerarm_l, loc=<17.8549 0 0>, rot=<0 0 0 -1>}
 deform_elbow_in_base_l {parent=lowerarm_l, loc=<0.2112 5.259 0.8274>, rot=<0 0 0 -1>}
 deform_elbow_in_l {parent=deform_elbow_in_base_l, loc=<0 0 0>, rot=<0 0 0 1>}
 deform_elbow_out_base_l {parent=lowerarm_l, loc=<0 -4 0>, rot=<0 0 0 -1>}
 deform_elbow_out_l {parent=deform_elbow_out_base_l, loc=<0 0 0>, rot=<0 0 0 1>}
 upperarm_twist_01_l {parent=upperarm_l, loc=<0.0041 0 0>, rot=<0 0 0 -1>}
 deform_upperarm_bk_base_l {parent=upperarm_twist_01_l, loc=<1.0813 -7.2193 -1.9453>, rot=<0 0 0 1>}
 deform_upperarm_bk_l {parent=deform_upperarm_bk_base_l, loc=<0 0 0>, rot=<0 0 0 1>}
 deform_upperarm_fr_base_l {parent=upperarm_twist_01_l, loc=<-0.3691 7.0506 -2.765>, rot=<0 0 0 1>}
 deform_upperarm_fr_l {parent=deform_upperarm_fr_base_l, loc=<0 0 0>, rot=<0 0 0 1>}
 deform_upperarm_up_base_l {parent=upperarm_twist_01_l, loc=<-3.2216 -1.1391 3.3311>, rot=<0 0 0 1>}
 deform_upperarm_up_l {parent=deform_upperarm_up_base_l, loc=<0 0 0>, rot=<0 0 0 1>}
 upperarm_twist_02_l {parent=upperarm_l, loc=<13.7941 0 0>, rot=<0 0 0 -1>}
 clavicle_r {parent=spine_05, loc=<4.3575 -0.1466 2.3223>, rot=<0.7289 0.0465 -0.6829 0.009>}
 upperarm_r {parent=clavicle_r, loc=<-15.5111 0 0>, rot=<0.0226 -0.3234 -0.0088 -0.9459>}
 lowerarm_r {parent=upperarm_r, loc=<-26.6514 0 0>, rot=<0.0391 -0.0099 -0.2453 -0.9686>}
 hand_r {parent=lowerarm_r, loc=<-22.3491 0 0>, rot=<-0.7081 -0.0507 0.0975 -0.6975>}
 thumb_01_r {parent=hand_r, loc=<-1.3118 0.3191 2.459>, rot=<0.6546 -0.4042 0.0664 -0.6354>}
 thumb_02_r {parent=thumb_01_r, loc=<-5.9336 0.0448 -0.0401>, rot=<-0.0037 0.0213 0.298 -0.9543>}
 thumb_03_r {parent=thumb_02_r, loc=<-2.2223 -0.3682 0.0907>, rot=<0.0118 -0.013 0.0066 -0.9998>}
 index_metacarpal_r {parent=hand_r, loc=<-1.6535 -0.0183 2.96>, rot=<0.0158 -0.0124 -0.0173 -0.9996>}
 index_01_r {parent=index_metacarpal_r, loc=<-8.3008 -0.0357 0.0524>, rot=<-0.0044 0.0485 0.0924 -0.9945>}
 index_02_r {parent=index_01_r, loc=<-3.3587 0.3004 -0.0452>, rot=<-0.002 0.0027 0.3225 -0.9466>}
 index_03_r {parent=index_02_r, loc=<-2.5578 0.123 0.0252>, rot=<-0.002 0.0067 0.2558 -0.9667>}
 middle_metacarpal_r {parent=hand_r, loc=<-2.1856 0.1725 0.5984>, rot=<0.006 0.0168 -0.0329 -0.9993>}
 middle_01_r {parent=middle_metacarpal_r, loc=<-7.3275 0.2447 -0.0129>, rot=<-0.0208 0.0042 0.0858 -0.9961>}
 middle_02_r {parent=middle_01_r, loc=<-4.0123 -0.1401 -0.1175>, rot=<-0.0218 0.0258 0.3512 -0.9357>}
 middle_03_r {parent=middle_02_r, loc=<-2.6077 -0.0534 -0.1373>, rot=<0.0082 0.0188 0.1708 -0.9851>}
 ring_metacarpal_r {parent=hand_r, loc=<-2.2074 0.3625 -1.3638>, rot=<-0.0284 0.0722 -0.0458 -0.9959>}
 ring_01_r {parent=ring_metacarpal_r, loc=<-7.0295 0.6197 0.1446>, rot=<-0.0489 -0.0252 0.1892 -0.9804>}
 ring_02_r {parent=ring_01_r, loc=<-3.6174 -0.4647 0.1027>, rot=<0 0 0.2806 -0.9598>}
 ring_03_r {parent=ring_02_r, loc=<-2.4152 -0.114 -0.0199>, rot=<0 0 0.1189 -0.9929>}
 pinky_metacarpal_r {parent=hand_r, loc=<-1.7616 0.9753 -3.3814>, rot=<-0.0397 0.0814 -0.069 -0.9935>}
 pinky_01_r {parent=pinky_metacarpal_r, loc=<-7.1709 0.4831 0.1094>, rot=<-0.0654 -0.0035 0.2242 -0.9723>}
 pinky_02_r {parent=pinky_01_r, loc=<-2.8274 -0.0893 0.1043>, rot=<0.0037 0.0011 0.2854 -0.9584>}
 pinky_03_r {parent=pinky_02_r, loc=<-2.2046 -0.0307 0.0078>, rot=<-0.0002 0.0098 0.1445 -0.9895>}
 deform_wrist_dn_base_r {parent=hand_r, loc=<-0.0004 2.2345 -0.0001>, rot=<0 0 0 -1>}
 deform_wrist_dn_r {parent=deform_wrist_dn_base_r, loc=<0 0 0>, rot=<0 0 0 1>}
 deform_wrist_up_base_r {parent=hand_r, loc=<0.0003 -2.2684 0.0001>, rot=<0 0 0 -1>}
 deform_wrist_up_r {parent=deform_wrist_up_base_r, loc=<0 0 0>, rot=<0 0 0 1>}
 lowerarm_twist_02_r {parent=lowerarm_r, loc=<-9.6199 0 0>, rot=<0 0 0 -1>}
 lowerarm_twist_01_r {parent=lowerarm_r, loc=<-17.8549 0 0>, rot=<0 0 0 -1>}
 deform_elbow_in_base_r {parent=lowerarm_r, loc=<-0.2111 -5.2591 -0.8275>, rot=<0 0 0 -1>}
 deform_elbow_in_r {parent=deform_elbow_in_base_r, loc=<-0.0001 0 0.0002>, rot=<0 0 0 1>}
 deform_elbow_out_base_r {parent=lowerarm_r, loc=<0 4 -0.0001>, rot=<0 0 0 -1>}
 deform_elbow_out_r {parent=deform_elbow_out_base_r, loc=<0 0 0>, rot=<0 0 0 1>}
 upperarm_twist_01_r {parent=upperarm_r, loc=<-0.0041 0 0>, rot=<0 0 0 -1>}
 deform_upperarm_bk_base_r {parent=upperarm_twist_01_r, loc=<-1.0814 7.2192 1.9455>, rot=<0 0 0 1>}
 deform_upperarm_bk_r {parent=deform_upperarm_bk_base_r, loc=<0 0 0>, rot=<0 0 0 1>}
 deform_upperarm_fr_base_r {parent=upperarm_twist_01_r, loc=<0.3689 -7.0506 2.7652>, rot=<0 0 0 1>}
 deform_upperarm_fr_r {parent=deform_upperarm_fr_base_r, loc=<0 0 0>, rot=<0 0 0 1>}
 deform_upperarm_up_base_r {parent=upperarm_twist_01_r, loc=<3.2214 1.1391 -3.3308>, rot=<0 0 0 1>}
 deform_upperarm_up_r {parent=deform_upperarm_up_base_r, loc=<0 0 0>, rot=<0 0 0 -1>}
 upperarm_twist_02_r {parent=upperarm_r, loc=<-13.7941 0 0>, rot=<0 0 0 -1>}
 deform_pec_aim_base_l {parent=spine_04, loc=<7.8732 12.6388 -2.4123>, rot=<0.0102 -0.4602 0.3979 -0.7936>}
 deform_pec_aim_l {parent=deform_pec_aim_base_l, loc=<0 0 0>, rot=<0 0 0 1>}
 deform_pec_base_l {parent=deform_pec_aim_l, loc=<10.766 0 0>, rot=<0 0 0 1>}
 deform_pec_l {parent=deform_pec_base_l, loc=<0.055 0 0>, rot=<0 0 0 1>}
 deform_pec_aim_base_r {parent=spine_04, loc=<7.8727 12.6388 2.4123>, rot=<-0.4602 -0.0102 0.7936 0.3979>}
 deform_pec_aim_r {parent=deform_pec_aim_base_r, loc=<0 0 0>, rot=<0 0 0 1>}
 deform_pec_base_r {parent=deform_pec_aim_r, loc=<-10.7662 -0.0006 -0.0007>, rot=<0 0 0 1>}
 deform_pec_r {parent=deform_pec_base_r, loc=<0 0 0>, rot=<0 0 0 1>}
 thigh_l {parent=pelvis, loc=<-2.5693 -0.0109 -7.7501>, rot=<0.0012 0.0572 0.0155 -0.9982>}
 calf_l {parent=thigh_l, loc=<-39.0484 0 0>, rot=<0.0006 -0.0161 -0.0391 -0.9991>}
 foot_l {parent=calf_l, loc=<-42.561 0.1077 0.4292>, rot=<0.0418 -0.0455 0.0343 -0.9975>}
 ball_l {parent=foot_l, loc=<-6.1599 13.8264 0.6806>, rot=<0.0001 -0.0001 -0.7063 -0.7079>}
 dyn_flip_flopper_l {parent=ball_l, loc=<-3.4615 1.3514 0>, rot=<0 0 0 1>}
 calf_twist_02_l {parent=calf_l, loc=<-18.8064 -0.0967 0.0397>, rot=<0.0237 0.0095 -0.0035 -0.9997>}
 calf_twist_01_l {parent=calf_l, loc=<-35.813 -0.0967 0.0397>, rot=<0.0237 0.0095 -0.0035 -0.9997>}
 deform_knee_in_base_l {parent=calf_l, loc=<1.298 -5.2957 -0.0496>, rot=<0 0 0 -1>}
 deform_knee_in_l {parent=deform_knee_in_base_l, loc=<0 0 0>, rot=<0 0 0 1>}
 deform_knee_out_base_l {parent=calf_l, loc=<0 4 0>, rot=<0 0 0 -1>}
 deform_knee_out_l {parent=deform_knee_out_base_l, loc=<0 0 0>, rot=<0 0 0 1>}
 dyn_calf_cloth_collider_l {parent=calf_l, loc=<0 0 0>, rot=<0 0 0 1>}
 thigh_twist_01_l {parent=thigh_l, loc=<-21.6303 0 0>, rot=<0.0013 0 0 -1>}
 dyn_shorts_base_l {parent=thigh_twist_01_l, loc=<0 0 0>, rot=<0 0 0 1>}
 dyn_shorts_04_l {parent=dyn_shorts_base_l, loc=<3.9999 1.1824 7.4629>, rot=<-0.0199 -0.6743 -0.7372 0.0364>}
 dyn_shorts_03_l {parent=dyn_shorts_base_l, loc=<3.8721 -9.5108 0.1904>, rot=<-0.0062 0.025 0.9995 -0.0186>}
 dyn_shorts_02_l {parent=dyn_shorts_base_l, loc=<4.1057 1.0113 -7.7269>, rot=<0.0349 0.7647 -0.6423 0.0368>}
 dyn_shorts_01_l {parent=dyn_shorts_base_l, loc=<4.9203 9.9126 -0.7511>, rot=<0 -1 0 -0.0073>}
 deform_groin_base_l {parent=thigh_l, loc=<-2.1908 8.1142 -1.7442>, rot=<0 0 0 -1>}
 deform_groin_l {parent=deform_groin_base_l, loc=<0 0 0>, rot=<0 0 0 1>}
 thigh_r {parent=pelvis, loc=<-2.5693 -0.0109 7.7501>, rot=<0.0572 -0.0012 0.9982 0.0155>}
 calf_r {parent=thigh_r, loc=<39.0484 0 0>, rot=<0.0006 -0.0161 -0.0391 -0.9991>}
 foot_r {parent=calf_r, loc=<42.561 -0.1077 -0.4292>, rot=<0.0418 -0.0455 0.0343 -0.9975>}
 ball_r {parent=foot_r, loc=<6.1599 -13.8264 -0.6806>, rot=<0.0001 -0.0001 -0.7063 -0.7079>}
 dyn_flip_flopper_r {parent=ball_r, loc=<3.4615 -1.3514 0>, rot=<0 0 0 -1>}
 calf_twist_02_r {parent=calf_r, loc=<18.8064 0.0967 -0.0397>, rot=<0.0237 0.0095 -0.0035 -0.9997>}
 calf_twist_01_r {parent=calf_r, loc=<35.813 0.0967 -0.0397>, rot=<0.0237 0.0095 -0.0035 -0.9997>}
 deform_knee_in_base_r {parent=calf_r, loc=<-1.298 5.2957 0.0496>, rot=<0 0 0 -1>}
 deform_knee_in_r {parent=deform_knee_in_base_r, loc=<0 0 0>, rot=<0 0 0 1>}
 deform_knee_out_base_r {parent=calf_r, loc=<0 -4 0>, rot=<0 0 0 -1>}
 deform_knee_out_r {parent=deform_knee_out_base_r, loc=<0 0 0>, rot=<0 0 0 1>}
 dyn_calf_cloth_collider_r {parent=calf_r, loc=<0 0 0>, rot=<0 0 0 1>}
 thigh_twist_01_r {parent=thigh_r, loc=<21.6303 0 0>, rot=<0.0013 0 0 -1>}
 dyn_shorts_base_r {parent=thigh_twist_01_r, loc=<0 0 0>, rot=<0 0 0 1>}
 dyn_shorts_04_r {parent=dyn_shorts_base_r, loc=<-3.9999 -1.1824 -7.4629>, rot=<0.0199 0.6743 0.7372 -0.0364>}
 dyn_shorts_03_r {parent=dyn_shorts_base_r, loc=<-3.8721 9.5108 -0.1904>, rot=<0.0062 -0.025 -0.9995 0.0186>}
 dyn_shorts_02_r {parent=dyn_shorts_base_r, loc=<-4.1057 -1.0113 7.7269>, rot=<-0.0349 -0.7647 0.6423 -0.0368>}
 dyn_shorts_01_r {parent=dyn_shorts_base_r, loc=<-4.9203 -9.9126 0.7511>, rot=<0 1 0 0.0073>}
 deform_groin_base_r {parent=thigh_r, loc=<2.1908 -8.1142 1.7442>, rot=<0 0 0 -1>}
 deform_groin_r {parent=deform_groin_base_r, loc=<0 0 0>, rot=<0 0 0 1>}
 deform_glute_base_l {parent=pelvis, loc=<-8.8064 -8.8222 -8.4287>, rot=<0.0275 0.0501 -0.4472 -0.8926>}
 deform_glute_l {parent=deform_glute_base_l, loc=<0 0 0>, rot=<0 0 0 1>}
 deform_glute_base_r {parent=pelvis, loc=<-8.8062 -8.8222 8.4287>, rot=<0.0501 -0.0275 0.8926 -0.4472>}
 deform_glute_r {parent=deform_glute_base_r, loc=<0 0 0>, rot=<0 0 0 1>}
 dyn_short_string_base_l {parent=pelvis, loc=<6.6485 9.3372 -2.924>, rot=<0 0 0 1>}
 dyn_short_string_01_l {parent=dyn_short_string_base_l, loc=<-0.1106 1.3131 -0.095>, rot=<-0.0085 -0.7071 -0.7071 0.0085>}
 dyn_short_string_02_l {parent=dyn_short_string_01_l, loc=<4.8701 0.0001 0.1597>, rot=<0 0 0 1>}
 dyn_short_string_03_l {parent=dyn_short_string_02_l, loc=<4.7429 -0.1213 0.2301>, rot=<0 0 0 1>}
 dyn_short_string_base_r {parent=pelvis, loc=<6.6485 9.3372 2.924>, rot=<0 0 1 0>}
 dyn_short_string_01_r {parent=dyn_short_string_base_r, loc=<0.1106 -1.3131 0.095>, rot=<-0.0085 -0.7071 -0.7071 0.0085>}
 dyn_short_string_02_r {parent=dyn_short_string_01_r, loc=<-4.8701 -0.0001 -0.1597>, rot=<0 0 0 1>}
 dyn_short_string_03_r {parent=dyn_short_string_02_r, loc=<-4.7429 0.1213 -0.2301>, rot=<0 0 0 1>}
 deform_belt_root {parent=pelvis, loc=<6.9039 -3.6811 14.7722>, rot=<-0.1514 -0.751 -0.1564 -0.6234>}
 deform_belt_front_01 {parent=deform_belt_root, loc=<0.2329 1.0092 0.0455>, rot=<0.6781 0.2892 0.2644 0.6218>}
 deform_belt_rear_01 {parent=deform_belt_root, loc=<0.1357 -1.1082 -0.1274>, rot=<0.5888 -0.3689 -0.4098 0.591>}
 deform_belt_front_02 {parent=deform_belt_root, loc=<7.8427 8.9265 0.0346>, rot=<0.7424 0.0415 0.0699 0.665>}
 deform_belt_front_03 {parent=deform_belt_root, loc=<14.854 10.0246 0.3802>, rot=<0.7503 -0.0732 -0.0512 0.655>}
 deform_belt_front_04 {parent=deform_belt_root, loc=<21.3076 8.8645 0.5058>, rot=<0.77 -0.1412 -0.1168 0.6112>}
 deform_belt_rear_02 {parent=deform_belt_root, loc=<3.308 -8.5459 -0.504>, rot=<0.6197 -0.1878 -0.2678 0.7135>}
 deform_belt_rear_03 {parent=deform_belt_root, loc=<11.1781 -14.7021 -1.1435>, rot=<0.6676 -0.0919 -0.1823 0.716>}
 deform_belt_rear_04 {parent=deform_belt_root, loc=<18.1981 -17.641 -1.9993>, rot=<0.6658 -0.0059 -0.1039 0.7388>}
 cr_spline_belt_rear_01 {parent=deform_belt_root, loc=<0.1357 -1.1082 -0.1274>, rot=<0.5888 -0.3689 -0.4098 0.591>}
 cr_spline_belt_rear_02 {parent=deform_belt_root, loc=<3.308 -8.5459 -0.504>, rot=<0.6197 -0.1878 -0.2678 0.7135>}
 cr_spline_belt_rear_03 {parent=deform_belt_root, loc=<11.1781 -14.7021 -1.1435>, rot=<0.6676 -0.0919 -0.1823 0.716>}
 cr_spline_belt_rear_04 {parent=deform_belt_root, loc=<18.1981 -17.641 -1.9993>, rot=<0.6658 -0.0059 -0.1039 0.7388>}
 cr_spline_belt_front_01 {parent=deform_belt_root, loc=<0.2329 1.0092 0.0455>, rot=<0.6781 0.2892 0.2644 0.6218>}
 cr_spline_belt_front_02 {parent=deform_belt_root, loc=<7.8427 8.9265 0.0346>, rot=<0.7424 0.0415 0.0699 0.665>}
 cr_spline_belt_front_03 {parent=deform_belt_root, loc=<14.854 10.0246 0.3802>, rot=<0.7503 -0.0732 -0.0512 0.655>}
 cr_spline_belt_front_04 {parent=deform_belt_root, loc=<21.3076 8.8645 0.5058>, rot=<0.77 -0.1412 -0.1168 0.6112>}
 cr_spline_result_belt_rear_01 {parent=deform_belt_root, loc=<0.1357 -1.1082 -0.1274>, rot=<0.5888 -0.3689 -0.4098 0.591>}
 cr_spline_result_belt_rear_02 {parent=deform_belt_root, loc=<3.308 -8.5459 -0.504>, rot=<0.6197 -0.1878 -0.2678 0.7135>}
 cr_spline_result_belt_rear_03 {parent=deform_belt_root, loc=<11.1781 -14.7021 -1.1435>, rot=<0.6676 -0.0919 -0.1823 0.716>}
 cr_splineresult_belt_rear_04 {parent=deform_belt_root, loc=<18.1981 -17.641 -1.9993>, rot=<0.6658 -0.0059 -0.1039 0.7388>}
 cr_spline_result_belt_front_01 {parent=deform_belt_root, loc=<0.2329 1.0092 0.0455>, rot=<0.6781 0.2892 0.2644 0.6218>}
 cr_spline_result_belt_front_02 {parent=deform_belt_root, loc=<7.8427 8.9265 0.0346>, rot=<0.7424 0.0415 0.0699 0.665>}
 cr_spline_result_belt_front_03 {parent=deform_belt_root, loc=<14.854 10.0246 0.3802>, rot=<0.7503 -0.0732 -0.0512 0.655>}
 cr_spline_result_belt_front_04 {parent=deform_belt_root, loc=<21.3076 8.8645 0.5058>, rot=<0.77 -0.1412 -0.1168 0.6112>}
 deform_belt_pouch {parent=pelvis, loc=<-1.6699 4.3325 -20.2909>, rot=<-0.6187 -0.021 0.7823 0.069>}
 deform_belt_pouchSocket {parent=deform_belt_pouch, type=Null, loc=<-1.4799 0.6512 0.9294>, rot=<0.9907 0.0632 -0.1183 0.0218>}
 dyn_belt_01 {parent=pelvis, loc=<5.2834 0 10.6592>, rot=<-0.4846 -0.5786 0.5161 -0.405>}
 dyn_belt_02 {parent=dyn_belt_01, loc=<8.8231 0.1533 -0.059>, rot=<0.7258 -0.0114 -0.0047 0.6878>}
 dyn_belt_03 {parent=dyn_belt_02, loc=<8.9729 0 0>, rot=<0 0 0 1>}
 dyn_belt_04 {parent=dyn_belt_03, loc=<14.1259 0 0>, rot=<0 0 0 1>}
 dyn_belt_linear_bounce {parent=dyn_belt_04, loc=<0 0 0>, rot=<0 0 0 1>}
 dyn_shorts_string_collider {parent=pelvis, loc=<0 0 0>, rot=<0 0 0 1>}
 dyn_flip_flopper_helper_l {parent=root, loc=<0 0 0>, rot=<0 0 0 1>}
 dyn_flip_flopper_helper_r {parent=root, loc=<0 0 0>, rot=<0 0 0 1>}
 Deform_Upperarms_Raised {type=Curve, loc=<0 0 0>, rot=<0 0 0 1>}
]
variables [
 ClothEnabled {type=bool, default=False}
 IsCrouch {type=bool, default=False}
 PawnVelocity {type=double, default=440}
 RBANEnabled {type=bool, default=False}
]
