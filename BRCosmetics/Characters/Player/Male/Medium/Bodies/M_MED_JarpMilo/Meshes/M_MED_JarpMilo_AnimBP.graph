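# Animation Blueprint for the JarpMilo player body, adding secondary motion to shorts, belt and cloth over the parent mesh's copied pose.
# A Control Rig feeds AnimDynamics shorts bones, bypassed when cloth is enabled, then two RigidBody simulations, a belt-pouch constraint and backpack attach offset.
# With cloth enabled, updates run the cloth controller and set low velocity scales during non-skydive falls or fast yaw.
# Events track skydiving with a backbling.
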AnimBlueprint /BRCosmetics/Characters/Player/Male/Medium/Bodies/M_MED_JarpMilo/Meshes/M_MED_JarpMilo_AnimBP
target_skeleton=&/M_MED_JarpMilo_Skeleton
graphs [
 AnimGraph {nodes [
   n1 = "Constraint - deform_belt_pouch (P)"(ComponentPose=Local To Component(LocalPose=Use cached pose 'Post RBAN'()), Alpha=1, ConstraintWeights_0=1) {
    alphaBoolBlend {blendOption=Linear}
    boneToModify {boneName=deform_belt_pouch}
    constraintSetup [{targetBone {boneName=dyn_belt_linear_bounce}}]
    constraintWeights [1]
   }
   n2 = "Transform (Modify) Bone - Bone: attach_backpack"!(ComponentPose=@n1, bAlphaBoolEnabled:bool=True, Translation="0,-1,-1.5", Rotation="0,0,10", Scale="1,1,1") -> (Pose) {
    alphaInputType=Bool
    rotationMode=BMM_Additive
    rotationSpace=BCS_BoneSpace
    scaleMode=BMM_Additive
    scaleSpace=BCS_BoneSpace
    translationMode=BMM_Additive
    translationSpace=BCS_BoneSpace
    alphaBoolBlend {blendInTime=0.1, blendOption=Sinusoidal, blendOutTime=0.1}
    boneToModify {boneName=attach_backpack}
    rotation {roll=10}
    translation {y=-1, z=-1.5}
   }
   Output Pose(Result=Component To Local(ComponentPose=@n2)) {layerGroup=None}
   n3 = Control Rig!(
    Source=Copy Pose From Mesh() {bUseAttachedParent=true}
    Alpha
    PawnVelocity:double=Get PawnSpeed2D()
    IsCrouch:bool=Get bIsCrouching()
    ClothEnabled:bool=Get bClothEnabled()
    RBANEnabled:bool=Get bRigidBodyEnabled()
   ) -> (Pose) {
    bTransferPoseInGlobalSpace=true
    alphaBoolBlend {blendOption=Linear}
    controlRigAssetReference {blueprintRigClass=&/M_MED_JarpMilo_CtrlRig.M_MED_JarpMilo_CtrlRig_C}
   }
   Save cached pose 'Deformation Rig'(Pose=@n3) {cacheName=Deformation Rig}
   n4 = Anim Dynamics - Bone: dyn_shorts_01_l!(
    SimulationSpace:AnimPhysSimSpaceType Enum=Component
    ComponentPose=Local To Component(LocalPose=Use cached pose 'Deformation Rig'())
    Alpha=1
    ExternalForce="0,0,0"
   ) -> (Pose) {
    angularDampingOverride=0.9
    angularSpringConstant=4000
    bAngularSpring=true
    bOverrideAngularDamping=true
    bUseGravityOverride=true
    alphaBoolBlend {blendOption=Linear}
    boundBone {boneName=dyn_shorts_01_l}
    componentAppliedLinearAccClamp=<1400 1400 1400>
    componentLinearAccScale=<1 1 1>
    componentLinearVelScale=<1 1 1>
    gravityOverride {z=-980}
    physicsBodyDefinitions [{
      boundBone {boneName=dyn_shorts_01_l}
      boxExtents=<32 16 16>
      constraintSetup {angularLimitsMax=<15 15 15>, angularLimitsMin=<-5 -5 -2>, angularTarget=<0 0 0>}
      localJointOffset=<8 0 0>
     }]
   }
   n5 = Anim Dynamics - Bone: dyn_shorts_02_l!(SimulationSpace:AnimPhysSimSpaceType Enum=Component, ComponentPose=@n4, Alpha=1, ExternalForce="0,0,0") -> (Pose) {
    angularDampingOverride=0.9
    angularSpringConstant=4000
    bAngularSpring=true
    bOverrideAngularDamping=true
    bUseGravityOverride=true
    alphaBoolBlend {blendOption=Linear}
    boundBone {boneName=dyn_shorts_02_l}
    componentAppliedLinearAccClamp=<1400 1400 1400>
    componentLinearAccScale=<1 1 1>
    componentLinearVelScale=<1 1 1>
    gravityOverride {z=-980}
    physicsBodyDefinitions [{
      boundBone {boneName=dyn_shorts_02_l}
      boxExtents=<32 16 16>
      constraintSetup {angularLimitsMax=<15 15 15>, angularLimitsMin=<-5 -5 -2>, angularTarget=<0 0 0>}
      localJointOffset=<8 0 0>
     }]
   }
   n6 = Anim Dynamics - Bone: dyn_shorts_03_l!(SimulationSpace:AnimPhysSimSpaceType Enum=Component, ComponentPose=@n5, Alpha=1, ExternalForce="0,0,0") -> (Pose) {
    angularDampingOverride=0.9
    angularSpringConstant=4000
    bAngularSpring=true
    bOverrideAngularDamping=true
    bUseGravityOverride=true
    alphaBoolBlend {blendOption=Linear}
    boundBone {boneName=dyn_shorts_03_l}
    componentAppliedLinearAccClamp=<1400 1400 1400>
    componentLinearAccScale=<1 1 1>
    componentLinearVelScale=<1 1 1>
    gravityOverride {z=-980}
    physicsBodyDefinitions [{
      boundBone {boneName=dyn_shorts_03_l}
      boxExtents=<32 16 16>
      constraintSetup {angularLimitsMax=<15 15 15>, angularLimitsMin=<-5 -5 -2>, angularTarget=<0 0 0>}
      localJointOffset=<8 0 0>
     }]
   }
   n7 = Anim Dynamics - Bone: dyn_shorts_04_l!(SimulationSpace:AnimPhysSimSpaceType Enum=Component, ComponentPose=@n6, Alpha=1, ExternalForce="0,0,0") -> (Pose) {
    angularDampingOverride=0.9
    angularSpringConstant=4000
    bAngularSpring=true
    bOverrideAngularDamping=true
    bUseGravityOverride=true
    alphaBoolBlend {blendOption=Linear}
    boundBone {boneName=dyn_shorts_04_l}
    componentAppliedLinearAccClamp=<1400 1400 1400>
    componentLinearAccScale=<1 1 1>
    componentLinearVelScale=<1 1 1>
    gravityOverride {z=-980}
    physicsBodyDefinitions [{
      boundBone {boneName=dyn_shorts_04_l}
      boxExtents=<32 16 16>
      constraintSetup {angularLimitsMax=<15 15 15>, angularLimitsMin=<-5 -5 -2>, angularTarget=<0 0 0>}
      localJointOffset=<8 0 0>
     }]
   }
   n8 = Anim Dynamics - Bone: dyn_shorts_01_r!(SimulationSpace:AnimPhysSimSpaceType Enum=Component, ComponentPose=@n7, Alpha=1, ExternalForce="0,0,0") -> (Pose) {
    angularDampingOverride=0.9
    angularSpringConstant=4000
    bAngularSpring=true
    bOverrideAngularDamping=true
    bUseGravityOverride=true
    alphaBoolBlend {blendOption=Linear}
    boundBone {boneName=dyn_shorts_01_r}
    componentAppliedLinearAccClamp=<1400 1400 1400>
    componentLinearAccScale=<1 1 1>
    componentLinearVelScale=<1 1 1>
    gravityOverride {z=-980}
    physicsBodyDefinitions [{
      boundBone {boneName=dyn_shorts_01_r}
      boxExtents=<32 16 16>
      constraintSetup {angularLimitsMax=<15 15 15>, angularLimitsMin=<-5 -5 -2>, angularTarget=<0 0 0>}
      localJointOffset=<-8 0 0>
     }]
   }
   n9 = Anim Dynamics - Bone: dyn_shorts_02_r!(SimulationSpace:AnimPhysSimSpaceType Enum=Component, ComponentPose=@n8, Alpha=1, ExternalForce="0,0,0") -> (Pose) {
    angularDampingOverride=0.9
    angularSpringConstant=4000
    bAngularSpring=true
    bOverrideAngularDamping=true
    bUseGravityOverride=true
    alphaBoolBlend {blendOption=Linear}
    boundBone {boneName=dyn_shorts_02_r}
    componentAppliedLinearAccClamp=<1400 1400 1400>
    componentLinearAccScale=<1 1 1>
    componentLinearVelScale=<1 1 1>
    gravityOverride {z=-980}
    physicsBodyDefinitions [{
      boundBone {boneName=dyn_shorts_02_r}
      boxExtents=<32 16 16>
      constraintSetup {angularLimitsMax=<15 15 15>, angularLimitsMin=<-5 -5 -2>, angularTarget=<0 0 0>}
      localJointOffset=<-8 0 0>
     }]
   }
   n10 = Anim Dynamics - Bone: dyn_shorts_03_r!(SimulationSpace:AnimPhysSimSpaceType Enum=Component, ComponentPose=@n9, Alpha=1, ExternalForce="0,0,0") -> (Pose) {
    angularDampingOverride=0.9
    angularSpringConstant=4000
    bAngularSpring=true
    bOverrideAngularDamping=true
    bUseGravityOverride=true
    alphaBoolBlend {blendOption=Linear}
    boundBone {boneName=dyn_shorts_03_r}
    componentAppliedLinearAccClamp=<1400 1400 1400>
    componentLinearAccScale=<1 1 1>
    componentLinearVelScale=<1 1 1>
    gravityOverride {z=-980}
    physicsBodyDefinitions [{
      boundBone {boneName=dyn_shorts_03_r}
      boxExtents=<32 16 16>
      constraintSetup {angularLimitsMax=<15 15 15>, angularLimitsMin=<-5 -5 -2>, angularTarget=<0 0 0>}
      localJointOffset=<-8 0 0>
     }]
   }
   n11 = Anim Dynamics - Bone: dyn_shorts_04_r!(SimulationSpace:AnimPhysSimSpaceType Enum=Component, ComponentPose=@n10, Alpha=1, ExternalForce="0,0,0") -> (Pose) {
    angularDampingOverride=0.9
    angularSpringConstant=4000
    bAngularSpring=true
    bOverrideAngularDamping=true
    bUseGravityOverride=true
    alphaBoolBlend {blendOption=Linear}
    boundBone {boneName=dyn_shorts_04_r}
    componentAppliedLinearAccClamp=<1400 1400 1400>
    componentLinearAccScale=<1 1 1>
    componentLinearVelScale=<1 1 1>
    gravityOverride {z=-980}
    physicsBodyDefinitions [{
      boundBone {boneName=dyn_shorts_04_r}
      boxExtents=<32 16 16>
      constraintSetup {angularLimitsMax=<15 15 15>, angularLimitsMin=<-5 -5 -2>, angularTarget=<0 0 0>}
      localJointOffset=<-8 0 0>
     }]
   }
   Save cached pose 'Post Short Setup'(Pose=Blend Poses by bool(
     BlendPose_0=Use cached pose 'Deformation Rig'()
     BlendPose_1=Component To Local(ComponentPose=@n11)
     BlendTime_0=0
     BlendTime_1=0
     bActiveValue=Get bClothEnabled()
    ) {blendType=Linear}) {cacheName=Post Short Setup}
   n12 = RigidBody(ComponentPose=Local To Component(LocalPose=Use cached pose 'Post Short Setup'())) {
    bOverrideWorldGravity=true
    overridePhysicsAsset=&/M_MED_JarpMilo_Body_Physics
    simulationSpace=BaseBoneSpace
    alphaBoolBlend {blendOption=Linear}
    baseBoneRef {boneName=spine_05}
    componentAppliedLinearAccClamp=<0 0 0>
    overrideWorldGravity {z=-980}
    simSpaceSettings {
     dampingAlpha=0
     maxAngularAcceleration=30
     maxAngularVelocity=10
     maxLinearAcceleration=4000
     maxLinearVelocity=770
     velocityScaleZ=0.4
     worldAlpha=0.5
     externalLinearDragV=<0.5 0.5 0.5>
    }
   }
   n13 = RigidBody(ComponentPose=@n12) {
    bOverrideWorldGravity=true
    overridePhysicsAsset=&/M_MED_JarpMilo_Body_Belt_Short_Physics
    simulationSpace=BaseBoneSpace
    alphaBoolBlend {blendOption=Linear}
    baseBoneRef {boneName=pelvis}
    componentAppliedLinearAccClamp=<0 0 0>
    overrideWorldGravity {z=-980}
    simSpaceSettings {
     dampingAlpha=0
     maxAngularAcceleration=30
     maxAngularVelocity=10
     maxLinearAcceleration=3500
     maxLinearVelocity=770
     velocityScaleZ=0.5
     worldAlpha=1
     externalLinearDragV=<1 1 1>
    }
   }
   Save cached pose 'Post RBAN'(Pose=Component To Local(ComponentPose=@n13)) {cacheName=Post RBAN}
  ]}
 EventGraph {nodes [
   n1 = Event BlueprintUpdateAnimation()
   > Set Delta Time X(Delta Time X=@n1.DeltaTimeX)
   > Branch(Condition=Get bClothEnabled())
   > n2 = "Cloth Controller Function (Parameter Asset) (Nativized)"!(
    execute
    CharacterPartAnimInstance:Custom Character Part Anim Instance Object Reference=Self-Reference()
    ClothingAssetName:Name=M_MED_JarpMilo_Body_Clothing_0
    ClothCoreSettings:"Cloth Core Settings Structure (by ref)"=Get Cloth Core Settings() -> (Cloth Core Settings:Cloth Core Settings Structure)
    WindGustLocalSettings:"Wind Gust Settings Structure (by ref)"=Get Wind Gust Local Settings() -> (Wind Gust Local Settings:Wind Gust Settings Structure)
    ClothUpAxis:Vector="1,0,0"
    LinVelScaleFwdBwdClamp:float=0
    LinVelScaleUpDwnClamp:float=0.7
    DeltaTime:float=Get Delta Time X()
    GameTimeInSeconds:float=GetGameTimeInSeconds()
    StateLogicLayerName:Name=none
    Parameters:Fort Character Dynamics Parameters Object Reference=&/M_MED_JarpMilo_PhysicsParams
   ) -> (
    then
    OutCurrentState:Name=None
    OutClothParameters:Cloth Parameters Structure
    OutClothCoreSettings:Cloth Core Settings Structure
    OutWindGustSettings:Wind Gust Parameters Structure
    OutWindGustLocalSettings:Wind Gust Settings Structure
    OutWindVelocity:Vector
    OutLinearVelVector:Vector
   )
   > SetClothingInteractorIfNotSet()
   n3 = Get bIsSkyDiving()
   > n4 = Branch(Condition=AND Boolean(A=Get bIsFalling(), B=NOT Boolean(A=@n3)))
   > SetVelocityScale!(
    execute
    self:Chaos Clothing Interactor Object Reference=Get ChaosClothingInteractor() -> (ChaosClothingInteractor:Chaos Clothing Interactor Object Reference)
    LinearVelocityScale:Vector="0.1,0.1,0.1"
    AngularVelocityScale:float=0.15
    FictitiousAngularScale:float=0
   ) -> (then)
   n5 = BreakRotator(InRot=Get PawnRotationRate() -> (PawnRotationRate:Rotator))
   Branch(execute=@n4.else, Condition="float >= float"(A="Absolute (Float)"(A=@n5.Yaw), B=4))
   > SetVelocityScale(self=Get ChaosClothingInteractor(), LinearVelocityScale="0.1,0.1,0.1", AngularVelocityScale=0.15, FictitiousAngularScale=0)
   Event OnCharacterCustomizationLoaded()
   > Do Once()
   n6 = CheckForBackbling()
   > Set isWearingBackbling(isWearingBackbling=OR Boolean(A=@n6, B=@n6.bIsBackblingLong))
   n7 = Event OnBeginSkydiving()
   n8 = Branch(execute=[@n7.then, @n9.then], Condition=Get isWearingBackbling())
   > Set SkydiveAndBackbling(SkydiveAndBackbling:bool=true) -> (Output_Get:bool)
   Set SkydiveAndBackbling(execute=@n8.else, SkydiveAndBackbling=false)
   n9 = Event OnBeginSkydivingFromBus() {is=title K2Node_Event}
   Event OnEndSkydiving()
   > Set SkydiveAndBackbling(SkydiveAndBackbling=false)
   "To String (Name)"(InName=@n2)
   PrintString!(execute, InString=Hello, bPrintToScreen, bPrintToLog, TextColor, Duration, Key) -> (then)
  ]}
 SetClothingInteractorIfNotSet {nodes [
   SetClothingInteractorIfNotSet() {is=title K2Node_FunctionEntry}
   n1 = IsValid(Object=Get ChaosClothingInteractor())
   > Branch(Condition=AND Boolean(A=Get bClothEnabled(), B=NOT Boolean(A=@n1)))
   n2 = MakeLiteralString(Value=M_MED_JarpMilo_Body_Clothing_0)
   n3 = GetOwningComponent()
   > n4 = Cast To ChaosClothingInteractor!(
    execute
    Object=GetClothingInteractor!(
     self:Clothing Simulation Interactor Object Reference=GetClothingSimulationInteractor!(self:Skeletal Mesh Component Object Reference=@n3, ClothingAssetName:Name=None) -> (ReturnValue:Clothing Simulation Interactor Object Reference)
     ClothingAssetName:Name=String To Name(InString=@n2)
    ) -> (ReturnValue:Clothing Interactor Object Reference)
   ) -> (then, CastFailed, AsChaos Clothing Interactor:Chaos Clothing Interactor Object Reference) {is=title K2Node_DynamicCast}
   > Set ChaosClothingInteractor(ChaosClothingInteractor:Chaos Clothing Interactor Object Reference=@n4) -> (Output_Get:Chaos Clothing Interactor Object Reference)
  ]}
]
variables [
 AnimDynamics Alpha {type=number, default=1}
 ChaosClothingInteractor {type=ChaosClothingInteractor, default=null}
 Cloth Core Settings {
  type=ClothCoreSettings
  default {
   collisionThickness=1
   dampingCoefficient=0.1
   frictionCoefficient=0.2
   localDampingCoefficient=0.05
   numberIterations=2
   numberSubsteps=1
   selfCollisionThickness=0
   animDriveDamping=<0.15 1>
   animDriveStiffness=<0.15 1>
   areaStiffness=<1 1>
   bendingStiffness=<1 1>
   edgeStiffness=<1 1>
   tetherStiffness=<1 1>
  }
 }
 Delta Time X {type=number}
 isWearingBackbling {type=boolean}
 Simulation Space {type=AnimPhysSimSpaceType, default=RootRelative}
 SkydiveAndBackbling {type=boolean}
 Wind Gust Local Settings {type=WindGustSettings, default {elapsedTime=0, finalGustStrengthLocal=0, gustDuration=0, gustStrength=0, timeBetweenGusts=0}}
]
