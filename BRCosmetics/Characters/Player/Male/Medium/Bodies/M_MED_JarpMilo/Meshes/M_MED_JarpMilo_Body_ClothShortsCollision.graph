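# Physics asset for cloth shorts collision on the JarpMilo medium male player body, as its name indicates.
# It defines bodies for both thighs and for the left and right dynamic calf cloth collider bones.
# Each calf collider is paired with the thigh on the same side by a constraint with free swing and twist.
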PhysicsAsset /BRCosmetics/Characters/Player/Male/Medium/Bodies/M_MED_JarpMilo/Meshes/M_MED_JarpMilo_Body_ClothShortsCollision
bodies {dyn_calf_cloth_collider_l [], dyn_calf_cloth_collider_r [], thigh_l [], thigh_r []}
constraints [
 {
  bone1Name=dyn_calf_cloth_collider_l
  bone2Name=thigh_l
  swing1LimitDegrees=45
  swing1Motion=Free
  swing2LimitDegrees=45
  swing2Motion=Free
  twistLimitDegrees=45
  twistMotion=Free
 }
 {
  bone1Name=dyn_calf_cloth_collider_r
  bone2Name=thigh_r
  swing1LimitDegrees=45
  swing1Motion=Free
  swing2LimitDegrees=45
  swing2Motion=Free
  twistLimitDegrees=45
  twistMotion=Free
 }
]
