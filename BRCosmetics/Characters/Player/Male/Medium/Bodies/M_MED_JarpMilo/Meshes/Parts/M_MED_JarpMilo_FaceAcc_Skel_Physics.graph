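# Physics asset for the JarpMilo male medium character's face accessory mesh, giving its braids, hair bun, flyaways and earrings capsule bodies for secondary motion.
# Two six-link braid chains, a three-link bun chain and both earrings hang from the head.
# Bun tips and flyaways branch off the bun links.
# Narrow angular limits restrain these parts, earrings most loosely, while larger head and upper-spine capsules share an angularly free joint.
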PhysicsAsset /BRCosmetics/Characters/Player/Male/Medium/Bodies/M_MED_JarpMilo/Meshes/Parts/M_MED_JarpMilo_FaceAcc_Skel_Physics
bodies {
 dyn_braid_01_l [{shapeType=Capsule, length=2, radius=1, shapeName=dyn_braid_01_l_capsule, center=<2 0 0>, rotation=<rot 90 0 0>}]
 dyn_braid_01_r [{shapeType=Capsule, length=2, radius=1, shapeName=dyn_braid_01_r_capsule, center=<2 0 0>, rotation=<rot 90 0 0>}]
 dyn_braid_02_l [{shapeType=Capsule, length=2, radius=1, shapeName=dyn_braid_02_l_capsule, center=<2 0 0>, rotation=<rot 90 0 0>}]
 dyn_braid_02_r [{shapeType=Capsule, length=2, radius=1, shapeName=dyn_braid_02_r_capsule, center=<2 0 0>, rotation=<rot 90 0 0>}]
 dyn_braid_03_l [{shapeType=Capsule, length=2, radius=1, shapeName=dyn_braid_03_l_capsule, center=<2 0 0>, rotation=<rot 90 0 0>}]
 dyn_braid_03_r [{shapeType=Capsule, length=2, radius=1, shapeName=dyn_braid_03_r_capsule, center=<2 0 0>, rotation=<rot 90 0 0>}]
 dyn_braid_04_l [{shapeType=Capsule, length=2, radius=1, shapeName=dyn_braid_04_l_capsule, center=<2 0 0>, rotation=<rot 90 0 0>}]
 dyn_braid_04_r [{shapeType=Capsule, length=2, radius=1, shapeName=dyn_braid_04_r_capsule, center=<2 0 0>, rotation=<rot 90 0 0>}]
 dyn_braid_05_l [{shapeType=Capsule, length=2, radius=1, shapeName=dyn_braid_05_l_capsule, center=<2 0 0>, rotation=<rot 90 0 0>}]
 dyn_braid_05_r [{shapeType=Capsule, length=2, radius=1, shapeName=dyn_braid_05_r_capsule, center=<2 0 0>, rotation=<rot 90 0 0>}]
 dyn_braid_06_l [{shapeType=Capsule, length=2, radius=1, shapeName=dyn_braid_06_l_capsule, center=<2 0 0>, rotation=<rot 90 0 0>}]
 dyn_braid_06_r [{shapeType=Capsule, length=2, radius=1, shapeName=dyn_braid_06_r_capsule, center=<2 0 0>, rotation=<rot 90 0 0>}]
 dyn_earring_l [{shapeType=Capsule, length=1.5, radius=1, shapeName=dyn_earring_l_capsule, center=<1 0 0>, rotation=<rot 90 0 0>}]
 dyn_earring_r [{shapeType=Capsule, length=1.5, radius=1, shapeName=dyn_earring_r_capsule, center=<-1 0 0>, rotation=<rot 90 0 0>}]
 dyn_hair_bun_01 [{shapeType=Capsule, length=2.375, radius=1.9005, shapeName=dyn_hair_bun_01_capsule, center=<1.5 0 -1.3>, rotation=<rot 0 0 90>}]
 dyn_hair_bun_02 [{shapeType=Capsule, length=4.625, radius=5.005, shapeName=dyn_hair_bun_02_capsule, center=<4.6288 0 -1.4124>, rotation=<rot 0 0 90>}]
 dyn_hair_bun_03 [{shapeType=Capsule, length=5.0064, radius=6.5673, shapeName=dyn_hair_bun_03_capsule, center=<4.0046 0 0>, rotation=<rot 0 0 90>}]
 dyn_hair_bun_03_tip_c [{shapeType=Capsule, length=1.5, radius=1, shapeName=dyn_hair_bun_03_tip_c_capsule, center=<1.5 0 0>, rotation=<rot 90 0 0>}]
 dyn_hair_bun_03_tip_l [{shapeType=Capsule, length=1.5, radius=1, shapeName=dyn_hair_bun_03_tip_l_capsule, center=<1.5 0 0>, rotation=<rot 90 0 0>}]
 dyn_hair_bun_03_tip_r [{shapeType=Capsule, length=1.5, radius=1, shapeName=dyn_hair_bun_03_tip_r_capsule, center=<1.5 0 0>, rotation=<rot 90 0 0>}]
 dyn_hair_bun_lower_flyaway_01_l [{shapeType=Capsule, length=1.5, radius=1, shapeName=dyn_hair_bun_lower_flyaway_01_l_capsule, center=<1.5 0 0>, rotation=<rot 90 0 0>}]
 dyn_hair_bun_lower_flyaway_01_r [{shapeType=Capsule, length=1.5, radius=1, shapeName=dyn_hair_bun_lower_flyaway_01_r_capsule, center=<1.5 0 0>, rotation=<rot 90 0 0>}]
 dyn_hair_bun_lower_flyaway_02_l [{shapeType=Capsule, length=1.5, radius=1, shapeName=dyn_hair_bun_lower_flyaway_02_l_capsule, center=<1.5 0 0>, rotation=<rot 90 0 0>}]
 dyn_hair_bun_lower_flyaway_02_r [{shapeType=Capsule, length=1.5, radius=1, shapeName=dyn_hair_bun_lower_flyaway_02_r_capsule, center=<1.5 0 0>, rotation=<rot 90 0 0>}]
 dyn_hair_bun_lower_flyaway_03_r [{shapeType=Capsule, length=1.5, radius=1, shapeName=dyn_hair_bun_lower_flyaway_03_r_capsule, center=<1.5 0 0>, rotation=<rot 90 0 0>}]
 dyn_hair_bun_upper_flyaway_01_l [{shapeType=Capsule, length=1.5, radius=1, shapeName=dyn_hair_bun_upper_flyaway_01_l_capsule, center=<1.5 0 0>, rotation=<rot 90 0 0>}]
 dyn_hair_bun_upper_flyaway_01_r [{shapeType=Capsule, length=1.5, radius=1, shapeName=dyn_hair_bun_upper_flyaway_01_r_capsule, center=<1.5 0 0>, rotation=<rot 90 0 0>}]
 dyn_hair_bun_upper_flyaway_02_l [{shapeType=Capsule, length=1.5, radius=1, shapeName=dyn_hair_bun_upper_flyaway_02_l_capsule, center=<1.5 0 0>, rotation=<rot 90 0 0>}]
 dyn_hair_bun_upper_flyaway_02_r [{shapeType=Capsule, length=1.5, radius=1, shapeName=dyn_hair_bun_upper_flyaway_02_r_capsule, center=<1.5 0 0>, rotation=<rot 90 0 0>}]
 dyn_hair_bun_upper_flyaway_03_r [{shapeType=Capsule, length=1.5, radius=1, shapeName=dyn_hair_bun_upper_flyaway_03_r_capsule, center=<1.5 0 0>, rotation=<rot 90 0 0>}]
 head [
  {shapeType=Capsule, length=14.0249, radius=7.172, center=<0.8845 1.1823 0>, rotation=<rot 0 -88.7103 -90>}
  {shapeType=Capsule, length=8.4375, radius=6, shapeName=head_capsule, center=<1.2183 4.9107 0>, rotation=<rot 0 -90 -90>}
 ]
 spine_05 [{shapeType=Capsule, length=8.25, radius=11.75, shapeName=spine_05_capsule, rotation=<rot 0 -90 0>}]
}
constraints [
 {bone1Name=dyn_braid_01_l, bone2Name=head, swing1LimitDegrees=12, swing2LimitDegrees=12, twistLimitDegrees=12}
 {bone1Name=dyn_braid_02_l, bone2Name=dyn_braid_01_l, swing1LimitDegrees=12, swing2LimitDegrees=12, twistLimitDegrees=12}
 {bone1Name=dyn_braid_03_l, bone2Name=dyn_braid_02_l, swing1LimitDegrees=12, swing2LimitDegrees=12, twistLimitDegrees=12}
 {bone1Name=dyn_braid_04_l, bone2Name=dyn_braid_03_l, swing1LimitDegrees=12, swing2LimitDegrees=12, twistLimitDegrees=12}
 {bone1Name=dyn_braid_05_l, bone2Name=dyn_braid_04_l, swing1LimitDegrees=12, swing2LimitDegrees=12, twistLimitDegrees=12}
 {bone1Name=dyn_braid_06_l, bone2Name=dyn_braid_05_l, swing1LimitDegrees=12, swing2LimitDegrees=12, twistLimitDegrees=12}
 {bone1Name=dyn_braid_01_r, bone2Name=head, swing1LimitDegrees=12, swing2LimitDegrees=12, twistLimitDegrees=12}
 {bone1Name=dyn_braid_02_r, bone2Name=dyn_braid_01_r, swing1LimitDegrees=12, swing2LimitDegrees=12, twistLimitDegrees=12}
 {bone1Name=dyn_braid_03_r, bone2Name=dyn_braid_02_r, swing1LimitDegrees=12, swing2LimitDegrees=12, twistLimitDegrees=12}
 {bone1Name=dyn_braid_04_r, bone2Name=dyn_braid_03_r, swing1LimitDegrees=12, swing2LimitDegrees=12, twistLimitDegrees=12}
 {bone1Name=dyn_braid_05_r, bone2Name=dyn_braid_04_r, swing1LimitDegrees=12, swing2LimitDegrees=12, twistLimitDegrees=12}
 {bone1Name=dyn_braid_06_r, bone2Name=dyn_braid_05_r, swing1LimitDegrees=12, swing2LimitDegrees=12, twistLimitDegrees=12}
 {bone1Name=dyn_hair_bun_01, bone2Name=head, swing1LimitDegrees=16, swing2LimitDegrees=8, twistLimitDegrees=8}
 {bone1Name=dyn_hair_bun_02, bone2Name=dyn_hair_bun_01, swing1LimitDegrees=16, swing2LimitDegrees=8, twistLimitDegrees=8}
 {
  bone1Name=head
  bone2Name=spine_05
  swing1LimitDegrees=45
  swing1Motion=Free
  swing2LimitDegrees=45
  swing2Motion=Free
  twistLimitDegrees=45
  twistMotion=Free
 }
 {bone1Name=dyn_hair_bun_03, bone2Name=dyn_hair_bun_02, swing1LimitDegrees=16, swing2LimitDegrees=8, twistLimitDegrees=8}
 {bone1Name=dyn_hair_bun_03_tip_c, bone2Name=dyn_hair_bun_03, swing1LimitDegrees=8, swing2LimitDegrees=8, twistLimitDegrees=8}
 {bone1Name=dyn_hair_bun_03_tip_l, bone2Name=dyn_hair_bun_03, swing1LimitDegrees=8, swing2LimitDegrees=8, twistLimitDegrees=8}
 {bone1Name=dyn_hair_bun_03_tip_r, bone2Name=dyn_hair_bun_03, swing1LimitDegrees=8, swing2LimitDegrees=8, twistLimitDegrees=8}
 {bone1Name=dyn_hair_bun_lower_flyaway_01_l, bone2Name=dyn_hair_bun_02, swing1LimitDegrees=8, swing2LimitDegrees=8, twistLimitDegrees=8}
 {bone1Name=dyn_hair_bun_lower_flyaway_01_r, bone2Name=dyn_hair_bun_02, swing1LimitDegrees=8, swing2LimitDegrees=8, twistLimitDegrees=8}
 {bone1Name=dyn_hair_bun_lower_flyaway_02_l, bone2Name=dyn_hair_bun_02, swing1LimitDegrees=8, swing2LimitDegrees=8, twistLimitDegrees=8}
 {bone1Name=dyn_hair_bun_lower_flyaway_02_r, bone2Name=dyn_hair_bun_02, swing1LimitDegrees=8, swing2LimitDegrees=8, twistLimitDegrees=8}
 {
  bone1Name=dyn_hair_bun_lower_flyaway_03_r
  bone2Name=dyn_hair_bun_lower_flyaway_02_r
  swing1LimitDegrees=8
  swing2LimitDegrees=8
  twistLimitDegrees=8
 }
 {bone1Name=dyn_hair_bun_upper_flyaway_01_l, bone2Name=dyn_hair_bun_01, swing1LimitDegrees=8, swing2LimitDegrees=8, twistLimitDegrees=8}
 {bone1Name=dyn_hair_bun_upper_flyaway_01_r, bone2Name=dyn_hair_bun_01, swing1LimitDegrees=8, swing2LimitDegrees=8, twistLimitDegrees=8}
 {bone1Name=dyn_hair_bun_upper_flyaway_02_l, bone2Name=dyn_hair_bun_01, swing1LimitDegrees=8, swing2LimitDegrees=8, twistLimitDegrees=8}
 {bone1Name=dyn_hair_bun_upper_flyaway_02_r, bone2Name=dyn_hair_bun_01, swing1LimitDegrees=8, swing2LimitDegrees=8, twistLimitDegrees=8}
 {bone1Name=dyn_hair_bun_upper_flyaway_03_r, bone2Name=dyn_hair_bun_01, swing1LimitDegrees=8, swing2LimitDegrees=8, twistLimitDegrees=8}
 {bone1Name=dyn_earring_l, bone2Name=head, swing1LimitDegrees=24, swing2LimitDegrees=24, twistLimitDegrees=24}
 {bone1Name=dyn_earring_r, bone2Name=head, swing1LimitDegrees=24, swing2LimitDegrees=24, twistLimitDegrees=24}
]
