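# Animation Blueprint for the JarpMilo face accessory part, adding secondary physics motion on top of the pose copied from the attached parent mesh.
# A head-relative RigidBody node simulates it, then two Transform Bone nodes additively adjust the hair bun bones.
# Each update, a nativized parameter-asset function computes gravity override and rigid body settings into variables.
# Customization load checks for backbling, and skydive events maintain a combined skydiving-with-backbling flag.
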AnimBlueprint /BRCosmetics/Characters/Player/Male/Medium/Bodies/M_MED_JarpMilo/Meshes/Parts/M_MED_JarpMilo_FaceAcc_AnimBP
target_skeleton=&/M_MED_JarpMilo_FaceAcc_Skeleton
graphs [
 AnimGraph {nodes [
   n1 = RigidBody!(
    ComponentPose=Local To Component(LocalPose=Copy Pose From Mesh() {bUseAttachedParent=true})
    Alpha
    OverrideWorldGravity:Vector="0,0,-980"
    ExternalForce
    SimSpaceSettings:Sim Space Settings Structure=(WorldAlpha=0.5,VelocityScaleZ=0.8,DampingAlpha=0,MaxLinearVelocity=770,MaxAngularVelocity=4,MaxLinearAcceleration=3500,MaxAngularAcceleration=20,ExternalLinearDragV=(X=2,Y=2,Z=2),ExternalLinearVelocity=(X=0,Y=0,Z=0),ExternalAngularVelocity=(X=0,Y=0,Z=0))
   ) -> (Pose) {
    bOverrideWorldGravity=true
    overridePhysicsAsset=&/M_MED_JarpMilo_FaceAcc_Skel_Physics
    simulationSpace=BaseBoneSpace
    alphaBoolBlend {blendOption=Linear}
    baseBoneRef {boneName=head}
    componentAppliedLinearAccClamp=<0 0 0>
    overrideWorldGravity {z=-980}
    simSpaceSettings {
     dampingAlpha=0
     maxAngularAcceleration=20
     maxAngularVelocity=4
     maxLinearAcceleration=3500
     maxLinearVelocity=770
     velocityScaleZ=0.8
     worldAlpha=0.5
     externalLinearDragV=<2 2 2>
    }
   }
   n2 = "Transform (Modify) Bone - Bone: dyn_hair_bun_01"(ComponentPose=@n1, Alpha=1, Translation="0,0,0", Rotation="-2,0,0", Scale="1,1,1") {
    rotationMode=BMM_Additive
    rotationSpace=BCS_BoneSpace
    alphaBoolBlend {blendOption=Linear}
    boneToModify {boneName=dyn_hair_bun_01}
    rotation {pitch=-2}
   }
   n3 = "Transform (Modify) Bone - Bone: dyn_hair_bun_02"!(ComponentPose=@n2, bAlphaBoolEnabled:bool=True, Translation="0,0,0", Rotation="-30,0,0", Scale="0.95,0.95,0.95") -> (Pose) {
    alphaInputType=Bool
    rotationMode=BMM_Additive
    rotationSpace=BCS_BoneSpace
    scaleMode=BMM_Additive
    scaleSpace=BCS_BoneSpace
    translationMode=BMM_Additive
    translationSpace=BCS_BoneSpace
    alphaBoolBlend {blendInTime=0.2, blendOption=Sinusoidal, blendOutTime=0.2}
    boneToModify {boneName=dyn_hair_bun_02}
    rotation {pitch=-30}
    scale=<0.95 0.95 0.95>
   }
   Output Pose(Result=Component To Local(ComponentPose=@n3)) {layerGroup=None}
  ]}
 EventGraph {nodes [
   Event BlueprintUpdateAnimation()
   > n1 = "Gravity Override State Function (Parameter Asset) (Nativized)"(
    CharacterPartAnimInstance=Self-Reference()
    GameTimeInSeconds=GetGameTimeInSeconds()
    Parameters=&/M_MED_JarpMilo_FaceAcc_PhysicsParams
   )
   > Set Gravity Override(Gravity Override=@n1)
   > Set RBAN Settings Out(RBAN Settings Out=@n1.OutRigidBodyAnimNodeParameters)
   Event OnCharacterCustomizationLoaded()
   > Do Once()
   n2 = CheckForBackbling()
   > Set isWearingBackbling(isWearingBackbling=OR Boolean(A=@n2, B=@n2.bIsBackblingLong))
   n3 = Event OnBeginSkydiving()
   n4 = Branch(execute=[@n3.then, @n5.then], Condition=Get isWearingBackbling())
   > Set SkydiveAndBackbling(SkydiveAndBackbling:bool=true) -> (Output_Get:bool)
   Set SkydiveAndBackbling(execute=@n4.else, SkydiveAndBackbling=false)
   n5 = Event OnBeginSkydivingFromBus() {is=title K2Node_Event}
   Event OnEndSkydiving()
   > Set SkydiveAndBackbling(SkydiveAndBackbling=false)
  ]}
]
variables [
 Gravity Override {type=Vector}
 isWearingBackbling {type=boolean}
 RBAN Settings Out {type=FortRigidBodyAnimNodeParameters}
 SkydiveAndBackbling {type=boolean}
]
